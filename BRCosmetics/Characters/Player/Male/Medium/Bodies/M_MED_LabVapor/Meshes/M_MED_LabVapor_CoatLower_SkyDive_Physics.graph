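# Physics asset for the lower coat of the male medium LabVapor body, named for its skydive state.
# Capsules cover pelvis, legs and feet, plus six coat chains (front, side and back, left and right) running from small base bodies constrained to the pelvis.
# Some chain ends carry two capsules, and front tips are cross-linked to neighboring side tips.
# All constraints lock swing and twist, leaving their degree limits inactive.
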PhysicsAsset /BRCosmetics/Characters/Player/Male/Medium/Bodies/M_MED_LabVapor/Meshes/M_MED_LabVapor_CoatLower_SkyDive_Physics
bodies {
 calf_l [{shapeType=Capsule, length=28.4355, radius=8.995, shapeName=calf_l_capsule, center=<-21.8857 -1.9628 0>, rotation=<rot 0 89.9999 90>}]
 calf_r [{shapeType=Capsule, length=28.4355, radius=8.995, shapeName=calf_r_capsule, center=<21.8857 1.9628 0>, rotation=<rot 0 89.9999 -90>}]
 dyn_coat_bck_l_01 [{shapeType=Capsule, length=15, radius=6, shapeName=dyn_coat_bck_l_01_capsule, center=<0 0 6>, rotation=<rot 0 0 -90>}]
 dyn_coat_bck_l_02 [
  {shapeType=Capsule, length=15, radius=6, shapeName=dyn_coat_bck_l_02_capsule, center=<2 0 5>, rotation=<rot 0 0 -90>}
  {shapeType=Capsule, length=15, radius=6, shapeName=dyn_coat_bck_l_02_capsule, center=<22 0 6>, rotation=<rot 0 0 -90>}
 ]
 dyn_coat_bck_l_base [{shapeType=Capsule, length=3, radius=1.5, shapeName=dyn_coat_bck_l_base_capsule, center=<0 0 3.4693>, rotation=<rot 0 0 90>}]
 dyn_coat_bck_r_01 [{shapeType=Capsule, length=15, radius=6, shapeName=dyn_coat_bck_r_01_capsule, center=<0 0 2>, rotation=<rot 0 0 -90>}]
 dyn_coat_bck_r_02 [
  {shapeType=Capsule, length=14, radius=6, shapeName=dyn_coat_bck_r_02_capsule, center=<2 0 5>, rotation=<rot 0 0 -90>}
  {shapeType=Capsule, length=16, radius=6, shapeName=dyn_coat_bck_r_02_capsule, center=<22 0 6>, rotation=<rot 0 0 -90>}
 ]
 dyn_coat_bck_r_base [{shapeType=Capsule, length=3, radius=1.5, shapeName=dyn_coat_bck_r_base_capsule, center=<0 0 3.4693>, rotation=<rot 0 0 90>}]
 dyn_coat_fr_l_01 [{shapeType=Capsule, length=15, radius=6, shapeName=dyn_coat_fr_l_01_capsule, center=<0 0 3>, rotation=<rot 0 0 -90>}]
 dyn_coat_fr_l_02 [{shapeType=Capsule, length=15, radius=6, shapeName=dyn_coat_fr_l_02_capsule, center=<0 0 3.5>, rotation=<rot 0 0 -80>}]
 dyn_coat_fr_l_03 [{shapeType=Capsule, length=15, radius=6, shapeName=dyn_coat_fr_l_03_capsule, center=<2 0 3>, rotation=<rot 0 0 -80>}]
 dyn_coat_fr_l_base [{shapeType=Capsule, length=3, radius=1.5, shapeName=dyn_coat_fr_l_base_capsule, center=<0 0 3.4693>, rotation=<rot 0 0 90>}]
 dyn_coat_fr_r_01 [{shapeType=Capsule, length=15, radius=6, shapeName=dyn_coat_fr_r_01_capsule, center=<0 0 6>, rotation=<rot 0 0 -90>}]
 dyn_coat_fr_r_02 [{shapeType=Capsule, length=15, radius=6, shapeName=dyn_coat_fr_r_02_capsule, center=<2 0 4>, rotation=<rot 0 0 -90>}]
 dyn_coat_fr_r_03 [{shapeType=Capsule, length=15, radius=6, shapeName=dyn_coat_fr_r_03_capsule, center=<2 0 5>, rotation=<rot 0 0 90>}]
 dyn_coat_fr_r_base [{shapeType=Capsule, length=3, radius=1.5, shapeName=dyn_coat_fr_r_base_capsule, center=<0 0 3.4693>, rotation=<rot 0 0 90>}]
 dyn_coat_side_l_01 [{shapeType=Capsule, length=15, radius=6, shapeName=dyn_coat_side_l_01_capsule, center=<0 0 3>, rotation=<rot 0 0 -90>}]
 dyn_coat_side_l_02 [{shapeType=Capsule, length=15, radius=6, shapeName=dyn_coat_side_l_02_capsule, center=<0 0 5>, rotation=<rot 0 0 -90>}]
 dyn_coat_side_l_03 [{shapeType=Capsule, length=15, radius=6, shapeName=dyn_coat_side_l_03_capsule, center=<2 0 5>, rotation=<rot 0 0 90>}]
 dyn_coat_side_l_base [{shapeType=Capsule, length=3, radius=1.5, shapeName=dyn_coat_side_l_base_capsule, center=<0 0 3.4693>, rotation=<rot 0 0 90>}]
 dyn_coat_side_r_01 [{shapeType=Capsule, length=15, radius=6, shapeName=dyn_coat_side_r_01_capsule, center=<0 0 6>, rotation=<rot 0 0 -90>}]
 dyn_coat_side_r_02 [
  {shapeType=Capsule, length=15, radius=6, shapeName=dyn_coat_side_r_02_capsule, center=<0 0 5>, rotation=<rot 0 0 -90>}
  {shapeType=Capsule, length=16, radius=7, shapeName=dyn_coat_side_r_02_capsule, center=<20 0 5>, rotation=<rot 0 0 -90>}
 ]
 dyn_coat_side_r_base [{shapeType=Capsule, length=3, radius=1.5, shapeName=dyn_coat_side_r_base_capsule, center=<0 0 3.4693>, rotation=<rot 0 0 90>}]
 foot_l [{shapeType=Capsule, length=15.7892, radius=8, shapeName=foot_l_capsule, center=<-3.1667 5.4849 0>, rotation=<rot 0 7.5277 -90>}]
 foot_r [{shapeType=Capsule, length=15.7892, radius=8, shapeName=foot_r_capsule, center=<3.1667 -5.4849 0>, rotation=<rot 0 7.5277 90>}]
 pelvis [{shapeType=Capsule, length=15, radius=11, shapeName=pelvis_capsule, center=<5 0 0>, rotation=<rot 0 -90 0>}]
 thigh_l [{shapeType=Capsule, length=32, radius=12, shapeName=thigh_l_capsule, center=<-20 -0.2343 0>, rotation=<rot 90 180 -0.5669>}]
 thigh_r [{shapeType=Capsule, length=32, radius=12, shapeName=thigh_r_capsule, center=<20 0.2343 0>, rotation=<rot 90 180 179.4331>}]
}
constraints [
 {
  bone1Name=thigh_l
  bone2Name=pelvis
  swing1LimitDegrees=45
  swing1Motion=Locked
  swing2LimitDegrees=45
  swing2Motion=Locked
  twistLimitDegrees=45
  twistMotion=Locked
 }
 {
  bone1Name=calf_l
  bone2Name=thigh_l
  swing1LimitDegrees=45
  swing1Motion=Locked
  swing2LimitDegrees=45
  swing2Motion=Locked
  twistLimitDegrees=45
  twistMotion=Locked
 }
 {
  bone1Name=thigh_r
  bone2Name=pelvis
  swing1LimitDegrees=45
  swing1Motion=Locked
  swing2LimitDegrees=45
  swing2Motion=Locked
  twistLimitDegrees=45
  twistMotion=Locked
 }
 {
  bone1Name=calf_r
  bone2Name=thigh_r
  swing1LimitDegrees=45
  swing1Motion=Locked
  swing2LimitDegrees=45
  swing2Motion=Locked
  twistLimitDegrees=45
  twistMotion=Locked
 }
 {
  bone1Name=foot_r
  bone2Name=calf_r
  swing1LimitDegrees=45
  swing1Motion=Locked
  swing2LimitDegrees=45
  swing2Motion=Locked
  twistLimitDegrees=45
  twistMotion=Locked
 }
 {
  bone1Name=foot_l
  bone2Name=calf_l
  swing1LimitDegrees=45
  swing1Motion=Locked
  swing2LimitDegrees=45
  swing2Motion=Locked
  twistLimitDegrees=45
  twistMotion=Locked
 }
 {
  bone1Name=dyn_coat_bck_l_base
  bone2Name=pelvis
  swing1LimitDegrees=45
  swing1Motion=Locked
  swing2LimitDegrees=45
  swing2Motion=Locked
  twistLimitDegrees=45
  twistMotion=Locked
 }
 {
  bone1Name=dyn_coat_bck_l_01
  bone2Name=dyn_coat_bck_l_base
  swing1LimitDegrees=10
  swing1Motion=Locked
  swing2LimitDegrees=10
  swing2Motion=Locked
  twistLimitDegrees=2
  twistMotion=Locked
 }
 {
  bone1Name=dyn_coat_bck_l_02
  bone2Name=dyn_coat_bck_l_01
  swing1LimitDegrees=10
  swing1Motion=Locked
  swing2LimitDegrees=10
  swing2Motion=Locked
  twistLimitDegrees=2
  twistMotion=Locked
 }
 {
  bone1Name=dyn_coat_bck_r_base
  bone2Name=pelvis
  swing1LimitDegrees=45
  swing1Motion=Locked
  swing2LimitDegrees=45
  swing2Motion=Locked
  twistLimitDegrees=45
  twistMotion=Locked
 }
 {
  bone1Name=dyn_coat_bck_r_01
  bone2Name=dyn_coat_bck_r_base
  swing1LimitDegrees=10
  swing1Motion=Locked
  swing2LimitDegrees=10
  swing2Motion=Locked
  twistLimitDegrees=2
  twistMotion=Locked
 }
 {
  bone1Name=dyn_coat_bck_r_02
  bone2Name=dyn_coat_bck_r_01
  swing1LimitDegrees=10
  swing1Motion=Locked
  swing2LimitDegrees=10
  swing2Motion=Locked
  twistLimitDegrees=2
  twistMotion=Locked
 }
 {
  bone1Name=dyn_coat_fr_l_base
  bone2Name=pelvis
  swing1LimitDegrees=45
  swing1Motion=Locked
  swing2LimitDegrees=45
  swing2Motion=Locked
  twistLimitDegrees=45
  twistMotion=Locked
 }
 {
  bone1Name=dyn_coat_fr_l_01
  bone2Name=dyn_coat_fr_l_base
  swing1LimitDegrees=10
  swing1Motion=Locked
  swing2LimitDegrees=10
  swing2Motion=Locked
  twistLimitDegrees=2
  twistMotion=Locked
 }
 {
  bone1Name=dyn_coat_fr_l_02
  bone2Name=dyn_coat_fr_l_01
  swing1LimitDegrees=10
  swing1Motion=Locked
  swing2LimitDegrees=10
  swing2Motion=Locked
  twistLimitDegrees=2
  twistMotion=Locked
 }
 {
  bone1Name=dyn_coat_fr_l_03
  bone2Name=dyn_coat_fr_l_02
  swing1LimitDegrees=10
  swing1Motion=Locked
  swing2LimitDegrees=10
  swing2Motion=Locked
  twistLimitDegrees=2
  twistMotion=Locked
 }
 {
  bone1Name=dyn_coat_fr_r_base
  bone2Name=pelvis
  swing1LimitDegrees=45
  swing1Motion=Locked
  swing2LimitDegrees=45
  swing2Motion=Locked
  twistLimitDegrees=45
  twistMotion=Locked
 }
 {
  bone1Name=dyn_coat_fr_r_01
  bone2Name=dyn_coat_fr_r_base
  swing1LimitDegrees=10
  swing1Motion=Locked
  swing2LimitDegrees=10
  swing2Motion=Locked
  twistLimitDegrees=2
  twistMotion=Locked
 }
 {
  bone1Name=dyn_coat_fr_r_02
  bone2Name=dyn_coat_fr_r_01
  swing1LimitDegrees=10
  swing1Motion=Locked
  swing2LimitDegrees=10
  swing2Motion=Locked
  twistLimitDegrees=2
  twistMotion=Locked
 }
 {
  bone1Name=dyn_coat_fr_r_03
  bone2Name=dyn_coat_fr_r_02
  swing1LimitDegrees=10
  swing1Motion=Locked
  swing2LimitDegrees=10
  swing2Motion=Locked
  twistLimitDegrees=2
  twistMotion=Locked
 }
 {
  bone1Name=dyn_coat_side_l_base
  bone2Name=pelvis
  swing1LimitDegrees=45
  swing1Motion=Locked
  swing2LimitDegrees=45
  swing2Motion=Locked
  twistLimitDegrees=45
  twistMotion=Locked
 }
 {
  bone1Name=dyn_coat_side_l_01
  bone2Name=dyn_coat_side_l_base
  swing1LimitDegrees=10
  swing1Motion=Locked
  swing2LimitDegrees=10
  swing2Motion=Locked
  twistLimitDegrees=2
  twistMotion=Locked
 }
 {
  bone1Name=dyn_coat_side_l_02
  bone2Name=dyn_coat_side_l_01
  swing1LimitDegrees=10
  swing1Motion=Locked
  swing2LimitDegrees=10
  swing2Motion=Locked
  twistLimitDegrees=2
  twistMotion=Locked
 }
 {
  bone1Name=dyn_coat_side_l_03
  bone2Name=dyn_coat_side_l_02
  swing1LimitDegrees=10
  swing1Motion=Locked
  swing2LimitDegrees=10
  swing2Motion=Locked
  twistLimitDegrees=2
  twistMotion=Locked
 }
 {
  bone1Name=dyn_coat_side_r_base
  bone2Name=pelvis
  swing1LimitDegrees=45
  swing1Motion=Locked
  swing2LimitDegrees=45
  swing2Motion=Locked
  twistLimitDegrees=45
  twistMotion=Locked
 }
 {
  bone1Name=dyn_coat_side_r_01
  bone2Name=dyn_coat_side_r_base
  swing1LimitDegrees=10
  swing1Motion=Locked
  swing2LimitDegrees=10
  swing2Motion=Locked
  twistLimitDegrees=2
  twistMotion=Locked
 }
 {
  bone1Name=dyn_coat_side_r_02
  bone2Name=dyn_coat_side_r_01
  swing1LimitDegrees=10
  swing1Motion=Locked
  swing2LimitDegrees=10
  swing2Motion=Locked
  twistLimitDegrees=2
  twistMotion=Locked
 }
 {
  bone1Name=dyn_coat_side_l_03
  bone2Name=dyn_coat_fr_l_03
  swing1LimitDegrees=45
  swing1Motion=Locked
  swing2LimitDegrees=45
  swing2Motion=Locked
  twistLimitDegrees=45
  twistMotion=Locked
 }
 {
  bone1Name=dyn_coat_fr_r_03
  bone2Name=dyn_coat_side_r_02
  swing1LimitDegrees=45
  swing1Motion=Locked
  swing2LimitDegrees=45
  swing2Motion=Locked
  twistLimitDegrees=45
  twistMotion=Locked
 }
]
