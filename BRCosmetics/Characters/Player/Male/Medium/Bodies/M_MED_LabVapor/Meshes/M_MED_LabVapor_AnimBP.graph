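# Animation Blueprint for the M_MED_LabVapor player body cosmetic that simulates the coat's secondary motion.
# Two Anim Dynamics chains read cached PreAnimPose and drive front, side and back coat bones with sprung, angle-limited bodies.
# bIsPlayingEmote blends between them, selecting the stiffer chain with wider back-panel limits during emotes.
# Additive Transform Bone nodes after the blend offset coat bones to adjust the drape.
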AnimBlueprint /BRCosmetics/Characters/Player/Male/Medium/Bodies/M_MED_LabVapor/Meshes/M_MED_LabVapor_AnimBP
target_skeleton=&/M_MED_LabVapor_Skeleton
graphs [
 AnimGraph {nodes [
   "Transform (Modify) Bone - Bone: dyn_coat_side_l_01"(Alpha=1, Translation="-3,0,0", Rotation="-5,0,-5", Scale="1,1,1") {
    rotationMode=BMM_Additive
    translationMode=BMM_Additive
    alphaBoolBlend {blendOption=Linear}
    boneToModify {boneName=dyn_coat_side_l_01}
    rotation {pitch=-5, roll=-5}
    translation {x=-3}
   }
   n1 = "Transform (Modify) Bone - Bone: dyn_coat_fr_l_base"(Alpha=1, Translation="0,0,0", Rotation="10,0,0", Scale="1,1,1") {
    rotationMode=BMM_Additive
    translationMode=BMM_Additive
    alphaBoolBlend {blendOption=Linear}
    alphaScaleBiasClamp {bInterpResult=true, interpSpeedIncreasing=6}
    boneToModify {boneName=dyn_coat_fr_l_base}
    rotation {pitch=10}
   }
   "Transform (Modify) Bone - Bone: dyn_coat_fr_l_02"(ComponentPose=@n1, Alpha=1, Translation="0,0,0", Rotation="7.086,0.000,0.000", Scale="1,1,1") {
    rotationMode=BMM_Additive
    translationMode=BMM_Additive
    alphaBoolBlend {blendOption=Linear}
    boneToModify {boneName=dyn_coat_fr_l_02}
    rotation {pitch=7.086}
   }
   n2 = "Transform (Modify) Bone - Bone: dyn_coat_side_l_01"(Alpha=1, Translation="-2,0,0", Rotation="-5,0,-5", Scale="1,1,1") {
    rotationMode=BMM_Additive
    translationMode=BMM_Additive
    alphaBoolBlend {blendOption=Linear}
    boneToModify {boneName=dyn_coat_side_l_01}
    rotation {pitch=-5, roll=-5}
    translation {x=-2}
   }
   n3 = "Transform (Modify) Bone - Bone: dyn_coat_fr_r_02"(ComponentPose=@n2, Alpha=1, Translation="0,0,0", Rotation="2,0,0", Scale="1,1,1") {
    rotationMode=BMM_Additive
    translationMode=BMM_Additive
    alphaBoolBlend {blendOption=Linear}
    boneToModify {boneName=dyn_coat_fr_r_02}
    rotation {pitch=2}
   }
   n4 = "Transform (Modify) Bone - Bone: dyn_coat_bck_l_base"(ComponentPose=@n3, Alpha=1, Translation="0,0,0", Rotation="0.000,2.542,-10.000", Scale="1,1,1") {
    rotationMode=BMM_Additive
    translationMode=BMM_Additive
    alphaBoolBlend {blendOption=Linear}
    boneToModify {boneName=dyn_coat_bck_l_base}
    rotation {roll=-10, yaw=2.542}
   }
   n5 = "Transform (Modify) Bone - Bone: dyn_coat_bck_r_base"(ComponentPose=@n4, Alpha=1, Translation="0,0,0", Rotation="0,0,-10", Scale="1,1,1") {
    rotationMode=BMM_Additive
    translationMode=BMM_Additive
    alphaBoolBlend {blendOption=Linear}
    boneToModify {boneName=dyn_coat_bck_r_base}
    rotation {roll=-10}
   }
   n6 = "Transform (Modify) Bone - Bone: dyn_coat_bck_l_02"(ComponentPose=@n5, Alpha=1, Translation="0,0,0", Rotation="0,0,0", Scale="1,1,1") {rotationMode=BMM_Additive, translationMode=BMM_Additive, alphaBoolBlend {blendOption=Linear}, boneToModify {boneName=dyn_coat_bck_l_02}}
   n7 = "Transform (Modify) Bone - Bone: dyn_coat_bck_r_02"(ComponentPose=@n6, Alpha=1, Translation="0,0,0", Rotation="0,0,0", Scale="1,1,1") {rotationMode=BMM_Additive, translationMode=BMM_Additive, alphaBoolBlend {blendOption=Linear}, boneToModify {boneName=dyn_coat_bck_r_02}}
   n8 = "Transform (Modify) Bone - Bone: dyn_coat_fr_r_base"(ComponentPose=@n7, Alpha=1, Translation="-1,1,0", Rotation="8,0,-5", Scale="1,1,1") {
    rotationMode=BMM_Additive
    translationMode=BMM_Additive
    alphaBoolBlend {blendOption=Linear}
    boneToModify {boneName=dyn_coat_fr_r_base}
    rotation {pitch=8, roll=-5}
    translation=<-1 1>
   }
   n9 = "Transform (Modify) Bone - Bone: dyn_coat_fr_l_base"(ComponentPose=@n8, Alpha=1, Translation="0,0,0", Rotation="0,0,0", Scale="1,1,1") {rotationMode=BMM_Additive, translationMode=BMM_Additive, alphaBoolBlend {blendOption=Linear}, boneToModify {boneName=dyn_coat_fr_l_base}}
   n10 = "Transform (Modify) Bone - Bone: dyn_coat_fr_l_base"(ComponentPose=@n9, Alpha=1, Translation="0,-1,0", Rotation="0,0,5", Scale="1,1,1") {
    rotationMode=BMM_Additive
    translationMode=BMM_Additive
    alphaBoolBlend {blendOption=Linear}
    boneToModify {boneName=dyn_coat_fr_l_base}
    rotation {roll=5}
    translation {y=-1}
   }
   "Transform (Modify) Bone - Bone: dyn_coat_fr_r_01"(ComponentPose=@n10, Alpha=1, Translation="0,0,0", Rotation="0,0,0", Scale="1,1,1") {rotationMode=BMM_Additive, translationMode=BMM_Additive, alphaBoolBlend {blendOption=Linear}, boneToModify {boneName=dyn_coat_fr_r_01}}
   n11 = Anim Dynamics - Bone: dyn_coat_fr_l_01!(
    SimulationSpace:AnimPhysSimSpaceType Enum
    ComponentPose=Local To Component(LocalPose=Use cached pose 'PreAnimPose'())
    Alpha=1
    GravityOverride:Vector="0,0,-980"
    ExternalForce="0,0,0"
   ) -> (Pose) {
    angularDampingOverride=0.9999
    angularSpringConstant=5000
    bAngularSpring=true
    bOverrideAngularDamping=true
    bUseGravityOverride=true
    gravityScale=0.35
    simulationSpace=RootRelative
    alphaBoolBlend {blendOption=Linear}
    boundBone {boneName=dyn_coat_fr_l_01}
    componentAppliedLinearAccClamp=<2000 2000 2000>
    componentLinearAccScale=<1 1 1>
    componentLinearVelScale=<1 1 1>
    gravityOverride {z=-980}
    physicsBodyDefinitions [{
      collisionType=CustomSphere
      sphereCollisionRadius=2
      boundBone {boneName=dyn_coat_fr_l_01}
      boxExtents=<20 30 30>
      constraintSetup {angularLimitsMax=<0 0 2>, angularLimitsMin=<0 -30 -2>, angularTarget=<1 0 0>}
      localJointOffset=<2.5 0 0>
     }]
   }
   n12 = Anim Dynamics - Bone: dyn_coat_fr_l_02!(SimulationSpace:AnimPhysSimSpaceType Enum, ComponentPose=@n11, Alpha=1, GravityOverride:Vector="0,0,-1000", ExternalForce="0,0,0") -> (Pose) {
    angularDampingOverride=0.9999
    angularSpringConstant=5000
    bAngularSpring=true
    bOverrideAngularDamping=true
    bUseGravityOverride=true
    gravityScale=0.35
    simulationSpace=RootRelative
    alphaBoolBlend {blendOption=Linear}
    boundBone {boneName=dyn_coat_fr_l_02}
    componentAppliedLinearAccClamp=<2000 2000 2000>
    componentLinearAccScale=<1 1 1>
    componentLinearVelScale=<1 1 1>
    gravityOverride {z=-1000}
    physicsBodyDefinitions [{
      collisionType=CustomSphere
      sphereCollisionRadius=2
      boundBone {boneName=dyn_coat_fr_l_02}
      boxExtents=<20 30 30>
      constraintSetup {angularLimitsMax=<0 0 2>, angularLimitsMin=<0 -30 -2>, angularTarget=<1 0 0>}
      localJointOffset=<2.5 0 0>
     }]
   }
   n13 = Anim Dynamics - Bone: dyn_coat_fr_l_03!(SimulationSpace:AnimPhysSimSpaceType Enum, ComponentPose=@n12, Alpha=1, GravityOverride:Vector="0,0,-1000", ExternalForce="0,0,0") -> (Pose) {
    angularDampingOverride=0.9999
    angularSpringConstant=5000
    bAngularSpring=true
    bOverrideAngularDamping=true
    bUseGravityOverride=true
    gravityScale=0.35
    simulationSpace=RootRelative
    alphaBoolBlend {blendOption=Linear}
    boundBone {boneName=dyn_coat_fr_l_03}
    componentAppliedLinearAccClamp=<2000 2000 2000>
    componentLinearAccScale=<1 1 1>
    componentLinearVelScale=<1 1 1>
    gravityOverride {z=-1000}
    physicsBodyDefinitions [{
      collisionType=CustomSphere
      sphereCollisionRadius=2
      boundBone {boneName=dyn_coat_fr_l_03}
      boxExtents=<20 30 30>
      constraintSetup {angularLimitsMax=<0 0 2>, angularLimitsMin=<0 -30 -2>, angularTarget=<1 0 0>}
      localJointOffset=<2.5 0 0>
     }]
   }
   n14 = Anim Dynamics - Bone: dyn_coat_side_l_01!(SimulationSpace:AnimPhysSimSpaceType Enum, ComponentPose=@n13, Alpha=1, GravityOverride:Vector="0,0,-980", ExternalForce="0,0,0") -> (Pose) {
    angularDampingOverride=0.9999
    angularSpringConstant=5000
    bAngularSpring=true
    bUseGravityOverride=true
    gravityScale=0.35
    simulationSpace=RootRelative
    alphaBoolBlend {blendOption=Linear}
    boundBone {boneName=dyn_coat_side_l_01}
    componentAppliedLinearAccClamp=<2000 2000 2000>
    componentLinearAccScale=<1 1 1>
    componentLinearVelScale=<1 1 1>
    gravityOverride {z=-980}
    physicsBodyDefinitions [{
      collisionType=CustomSphere
      sphereCollisionRadius=2
      boundBone {boneName=dyn_coat_side_l_01}
      boxExtents=<20 30 30>
      constraintSetup {angularLimitsMax=<0 0 2>, angularLimitsMin=<0 -30 -2>, angularTarget=<1 0 0>}
      localJointOffset=<2.5 0 0>
     }]
   }
   n15 = Anim Dynamics - Bone: dyn_coat_side_l_02!(SimulationSpace:AnimPhysSimSpaceType Enum, ComponentPose=@n14, Alpha=1, GravityOverride:Vector="0,0,-1000", ExternalForce="0,0,0") -> (Pose) {
    angularDampingOverride=0.9999
    angularSpringConstant=5000
    bAngularSpring=true
    bOverrideAngularDamping=true
    bUseGravityOverride=true
    gravityScale=0.35
    simulationSpace=RootRelative
    alphaBoolBlend {blendOption=Linear}
    boundBone {boneName=dyn_coat_side_l_02}
    componentAppliedLinearAccClamp=<2000 2000 2000>
    componentLinearAccScale=<1 1 1>
    componentLinearVelScale=<1 1 1>
    gravityOverride {z=-1000}
    physicsBodyDefinitions [{
      collisionType=CustomSphere
      sphereCollisionRadius=2
      boundBone {boneName=dyn_coat_side_l_02}
      boxExtents=<20 30 30>
      constraintSetup {angularLimitsMax=<0 0 2>, angularLimitsMin=<0 -30 -2>, angularTarget=<1 0 0>}
      localJointOffset=<2.5 0 0>
     }]
   }
   n16 = Anim Dynamics - Bone: dyn_coat_side_l_03!(SimulationSpace:AnimPhysSimSpaceType Enum, ComponentPose=@n15, Alpha=1, GravityOverride:Vector="0,0,-1000", ExternalForce="0,0,0") -> (Pose) {
    angularDampingOverride=0.9999
    angularSpringConstant=5000
    bAngularSpring=true
    bOverrideAngularDamping=true
    bUseGravityOverride=true
    gravityScale=0.35
    simulationSpace=RootRelative
    alphaBoolBlend {blendOption=Linear}
    boundBone {boneName=dyn_coat_side_l_03}
    componentAppliedLinearAccClamp=<2000 2000 2000>
    componentLinearAccScale=<1 1 1>
    componentLinearVelScale=<1 1 1>
    gravityOverride {z=-1000}
    physicsBodyDefinitions [{
      collisionType=CustomSphere
      sphereCollisionRadius=2
      boundBone {boneName=dyn_coat_side_l_03}
      boxExtents=<20 30 30>
      constraintSetup {angularLimitsMax=<0 0 2>, angularLimitsMin=<0 -30 -2>, angularTarget=<1 0 0>}
      localJointOffset=<2.5 0 0>
     }]
   }
   n17 = Anim Dynamics - Bone: dyn_coat_fr_r_01!(SimulationSpace:AnimPhysSimSpaceType Enum, ComponentPose=@n16, Alpha=1, GravityOverride:Vector="0,0,-980", ExternalForce="0,0,0") -> (Pose) {
    angularDampingOverride=0.9999
    angularSpringConstant=5000
    bAngularSpring=true
    bUseGravityOverride=true
    gravityScale=0.35
    simulationSpace=RootRelative
    alphaBoolBlend {blendOption=Linear}
    boundBone {boneName=dyn_coat_fr_r_01}
    componentAppliedLinearAccClamp=<2000 2000 2000>
    componentLinearAccScale=<1 1 1>
    componentLinearVelScale=<1 1 1>
    gravityOverride {z=-980}
    physicsBodyDefinitions [{
      collisionType=CustomSphere
      sphereCollisionRadius=2
      boundBone {boneName=dyn_coat_fr_r_01}
      boxExtents=<20 30 30>
      constraintSetup {angularLimitsMax=<0 0 2>, angularLimitsMin=<0 -30 -2>, angularTarget=<1 0 0>}
      localJointOffset=<2.5 0 0>
     }]
   }
   n18 = Anim Dynamics - Bone: dyn_coat_fr_r_02!(SimulationSpace:AnimPhysSimSpaceType Enum, ComponentPose=@n17, Alpha=1, GravityOverride:Vector="0,0,-1000", ExternalForce="0,0,0") -> (Pose) {
    angularDampingOverride=0.9999
    angularSpringConstant=5000
    bAngularSpring=true
    bOverrideAngularDamping=true
    bUseGravityOverride=true
    gravityScale=0.35
    simulationSpace=RootRelative
    alphaBoolBlend {blendOption=Linear}
    boundBone {boneName=dyn_coat_fr_r_02}
    componentAppliedLinearAccClamp=<2000 2000 2000>
    componentLinearAccScale=<1 1 1>
    componentLinearVelScale=<1 1 1>
    gravityOverride {z=-1000}
    physicsBodyDefinitions [{
      collisionType=CustomSphere
      sphereCollisionRadius=2
      boundBone {boneName=dyn_coat_fr_r_02}
      boxExtents=<20 30 30>
      constraintSetup {angularLimitsMax=<0 0 2>, angularLimitsMin=<0 -30 -2>, angularTarget=<1 0 0>}
      localJointOffset=<2.5 0 0>
     }]
   }
   n19 = Anim Dynamics - Bone: dyn_coat_fr_r_03!(SimulationSpace:AnimPhysSimSpaceType Enum, ComponentPose=@n18, Alpha=1, GravityOverride:Vector="0,0,-1000", ExternalForce="0,0,0") -> (Pose) {
    angularDampingOverride=0.9999
    angularSpringConstant=5000
    bAngularSpring=true
    bOverrideAngularDamping=true
    bUseGravityOverride=true
    gravityScale=0.35
    simulationSpace=RootRelative
    alphaBoolBlend {blendOption=Linear}
    boundBone {boneName=dyn_coat_fr_r_03}
    componentAppliedLinearAccClamp=<2000 2000 2000>
    componentLinearAccScale=<1 1 1>
    componentLinearVelScale=<1 1 1>
    gravityOverride {z=-1000}
    physicsBodyDefinitions [{
      collisionType=CustomSphere
      sphereCollisionRadius=2
      boundBone {boneName=dyn_coat_fr_r_03}
      boxExtents=<20 30 30>
      constraintSetup {angularLimitsMax=<0 0 2>, angularLimitsMin=<0 -30 -2>, angularTarget=<1 0 0>}
      localJointOffset=<2.5 0 0>
     }]
   }
   n20 = Anim Dynamics - Bone: dyn_coat_side_r_01!(SimulationSpace:AnimPhysSimSpaceType Enum, ComponentPose=@n19, Alpha=1, GravityOverride:Vector="0,0,-980", ExternalForce="0,0,0") -> (Pose) {
    angularDampingOverride=0.9999
    angularSpringConstant=5000
    bAngularSpring=true
    bUseGravityOverride=true
    gravityScale=0.35
    simulationSpace=RootRelative
    alphaBoolBlend {blendOption=Linear}
    boundBone {boneName=dyn_coat_side_r_01}
    componentAppliedLinearAccClamp=<2000 2000 2000>
    componentLinearAccScale=<1 1 1>
    componentLinearVelScale=<1 1 1>
    gravityOverride {z=-980}
    physicsBodyDefinitions [{
      collisionType=CustomSphere
      sphereCollisionRadius=2
      boundBone {boneName=dyn_coat_side_r_01}
      boxExtents=<20 30 30>
      constraintSetup {angularLimitsMax=<0 0 2>, angularLimitsMin=<0 -30 -2>, angularTarget=<1 0 0>}
      localJointOffset=<2.5 0 0>
     }]
   }
   n21 = Anim Dynamics - Bone: dyn_coat_side_r_02!(SimulationSpace:AnimPhysSimSpaceType Enum, ComponentPose=@n20, Alpha=1, GravityOverride:Vector="0,0,-1000", ExternalForce="0,0,0") -> (Pose) {
    angularDampingOverride=0.9999
    angularSpringConstant=5000
    bAngularSpring=true
    bOverrideAngularDamping=true
    bUseGravityOverride=true
    gravityScale=0.35
    simulationSpace=RootRelative
    alphaBoolBlend {blendOption=Linear}
    boundBone {boneName=dyn_coat_side_r_02}
    componentAppliedLinearAccClamp=<2000 2000 2000>
    componentLinearAccScale=<1 1 1>
    componentLinearVelScale=<1 1 1>
    gravityOverride {z=-1000}
    physicsBodyDefinitions [{
      collisionType=CustomSphere
      sphereCollisionRadius=2
      boundBone {boneName=dyn_coat_side_r_02}
      boxExtents=<20 30 30>
      constraintSetup {angularLimitsMax=<0 0 2>, angularLimitsMin=<0 -30 -2>, angularTarget=<1 0 0>}
      localJointOffset=<2.5 0 0>
     }]
   }
   n22 = Anim Dynamics - Bone: dyn_coat_side_r_03!(SimulationSpace:AnimPhysSimSpaceType Enum, ComponentPose=@n21, Alpha=1, GravityOverride:Vector="0,0,-1000", ExternalForce="0,0,0") -> (Pose) {
    angularDampingOverride=0.9999
    angularSpringConstant=5000
    bAngularSpring=true
    bOverrideAngularDamping=true
    bUseGravityOverride=true
    gravityScale=0.35
    simulationSpace=RootRelative
    alphaBoolBlend {blendOption=Linear}
    boundBone {boneName=dyn_coat_side_r_03}
    componentAppliedLinearAccClamp=<2000 2000 2000>
    componentLinearAccScale=<1 1 1>
    componentLinearVelScale=<1 1 1>
    gravityOverride {z=-1000}
    physicsBodyDefinitions [{
      collisionType=CustomSphere
      sphereCollisionRadius=2
      boundBone {boneName=dyn_coat_side_r_03}
      boxExtents=<20 30 30>
      constraintSetup {angularLimitsMax=<0 0 2>, angularLimitsMin=<0 -30 -2>, angularTarget=<1 0 0>}
      localJointOffset=<2.5 0 0>
     }]
   }
   n23 = Anim Dynamics - Bone: dyn_coat_bck_l_01!(SimulationSpace:AnimPhysSimSpaceType Enum, ComponentPose=@n22, Alpha=1, GravityOverride:Vector="0,0,-980", ExternalForce="0,0,0") -> (Pose) {
    angularDampingOverride=0.9999
    angularSpringConstant=5000
    bAngularSpring=true
    bUseGravityOverride=true
    gravityScale=0.35
    simulationSpace=RootRelative
    alphaBoolBlend {blendOption=Linear}
    boundBone {boneName=dyn_coat_bck_l_01}
    componentAppliedLinearAccClamp=<2000 2000 2000>
    componentLinearAccScale=<1 1 1>
    componentLinearVelScale=<1 1 1>
    gravityOverride {z=-980}
    physicsBodyDefinitions [{
      collisionType=CustomSphere
      sphereCollisionRadius=2
      boundBone {boneName=dyn_coat_bck_l_01}
      boxExtents=<20 30 30>
      constraintSetup {angularLimitsMax=<0 15 2>, angularLimitsMin=<0 -30 -2>, angularTarget=<1 0 0>}
      localJointOffset=<2.5 0 0>
     }]
   }
   n24 = Anim Dynamics - Bone: dyn_coat_bck_l_02!(SimulationSpace:AnimPhysSimSpaceType Enum, ComponentPose=@n23, Alpha=1, GravityOverride:Vector="0,0,-1000", ExternalForce="0,0,0") -> (Pose) {
    angularDampingOverride=0.9999
    angularSpringConstant=5000
    bAngularSpring=true
    bOverrideAngularDamping=true
    bUseGravityOverride=true
    gravityScale=0.35
    simulationSpace=RootRelative
    alphaBoolBlend {blendOption=Linear}
    boundBone {boneName=dyn_coat_bck_l_02}
    componentAppliedLinearAccClamp=<2000 2000 2000>
    componentLinearAccScale=<1 1 1>
    componentLinearVelScale=<1 1 1>
    gravityOverride {z=-1000}
    physicsBodyDefinitions [{
      collisionType=CustomSphere
      sphereCollisionRadius=2
      boundBone {boneName=dyn_coat_bck_l_02}
      boxExtents=<20 30 30>
      constraintSetup {angularLimitsMax=<0 15 2>, angularLimitsMin=<0 -30 -2>, angularTarget=<1 0 0>}
      localJointOffset=<2.5 0 0>
     }]
   }
   n25 = Anim Dynamics - Bone: dyn_coat_bck_l_03!(SimulationSpace:AnimPhysSimSpaceType Enum, ComponentPose=@n24, Alpha=1, GravityOverride:Vector="0,0,-1000", ExternalForce="0,0,0") -> (Pose) {
    angularDampingOverride=0.9999
    angularSpringConstant=5000
    bAngularSpring=true
    bOverrideAngularDamping=true
    bUseGravityOverride=true
    gravityScale=0.35
    simulationSpace=RootRelative
    alphaBoolBlend {blendOption=Linear}
    boundBone {boneName=dyn_coat_bck_l_03}
    componentAppliedLinearAccClamp=<2000 2000 2000>
    componentLinearAccScale=<1 1 1>
    componentLinearVelScale=<1 1 1>
    gravityOverride {z=-1000}
    physicsBodyDefinitions [{
      collisionType=CustomSphere
      sphereCollisionRadius=2
      boundBone {boneName=dyn_coat_bck_l_03}
      boxExtents=<20 30 30>
      constraintSetup {angularLimitsMax=<0 15 2>, angularLimitsMin=<0 -30 -2>, angularTarget=<1 0 0>}
      localJointOffset=<2.5 0 0>
     }]
   }
   n26 = Anim Dynamics - Bone: dyn_coat_bck_r_01!(SimulationSpace:AnimPhysSimSpaceType Enum=Component, ComponentPose=@n25, Alpha=1, GravityOverride:Vector="0,0,-980", ExternalForce="0,0,0") -> (Pose) {
    angularDampingOverride=0.9999
    angularSpringConstant=5000
    bAngularSpring=true
    bUseGravityOverride=true
    gravityScale=0.35
    simulationSpace=RootRelative
    alphaBoolBlend {blendOption=Linear}
    boundBone {boneName=dyn_coat_bck_r_01}
    componentAppliedLinearAccClamp=<2000 2000 2000>
    componentLinearAccScale=<1 1 1>
    componentLinearVelScale=<1 1 1>
    gravityOverride {z=-980}
    physicsBodyDefinitions [{
      collisionType=CustomSphere
      sphereCollisionRadius=2
      boundBone {boneName=dyn_coat_bck_r_01}
      boxExtents=<20 30 30>
      constraintSetup {angularLimitsMax=<0 15 2>, angularLimitsMin=<0 -30 -2>, angularTarget=<1 0 0>}
      localJointOffset=<2.5 0 0>
     }]
   }
   n27 = Anim Dynamics - Bone: dyn_coat_bck_r_02!(SimulationSpace:AnimPhysSimSpaceType Enum, ComponentPose=@n26, Alpha=1, GravityOverride:Vector="0,0,-1000", ExternalForce="0,0,0") -> (Pose) {
    angularDampingOverride=0.9999
    angularSpringConstant=5000
    bAngularSpring=true
    bOverrideAngularDamping=true
    bUseGravityOverride=true
    gravityScale=0.35
    simulationSpace=RootRelative
    alphaBoolBlend {blendOption=Linear}
    boundBone {boneName=dyn_coat_bck_r_02}
    componentAppliedLinearAccClamp=<2000 2000 2000>
    componentLinearAccScale=<1 1 1>
    componentLinearVelScale=<1 1 1>
    gravityOverride {z=-1000}
    physicsBodyDefinitions [{
      collisionType=CustomSphere
      sphereCollisionRadius=2
      boundBone {boneName=dyn_coat_bck_r_02}
      boxExtents=<20 30 30>
      constraintSetup {angularLimitsMax=<0 15 2>, angularLimitsMin=<0 -30 -2>, angularTarget=<1 0 0>}
      localJointOffset=<2.5 0 0>
     }]
   }
   n28 = Anim Dynamics - Bone: dyn_coat_bck_r_03!(SimulationSpace:AnimPhysSimSpaceType Enum, ComponentPose=@n27, Alpha=1, GravityOverride:Vector="0,0,-1000", ExternalForce="0,0,0") -> (Pose) {
    angularDampingOverride=0.9999
    angularSpringConstant=5000
    bAngularSpring=true
    bOverrideAngularDamping=true
    bUseGravityOverride=true
    gravityScale=0.35
    simulationSpace=RootRelative
    alphaBoolBlend {blendOption=Linear}
    boundBone {boneName=dyn_coat_bck_r_03}
    componentAppliedLinearAccClamp=<2000 2000 2000>
    componentLinearAccScale=<1 1 1>
    componentLinearVelScale=<1 1 1>
    gravityOverride {z=-1000}
    physicsBodyDefinitions [{
      collisionType=CustomSphere
      sphereCollisionRadius=2
      boundBone {boneName=dyn_coat_bck_r_03}
      boxExtents=<20 30 30>
      constraintSetup {angularLimitsMax=<0 15 2>, angularLimitsMin=<0 -30 -2>, angularTarget=<1 0 0>}
      localJointOffset=<2.5 0 0>
     }]
   }
   n29 = Component To Local(ComponentPose=@n28)
   n30 = Anim Dynamics - Bone: dyn_coat_fr_l_01(ComponentPose=Local To Component(LocalPose=Use cached pose 'PreAnimPose'()), Alpha=1, GravityOverride="0,0,-980", ExternalForce="0,0,0") {
    angularDampingOverride=0.9999
    angularSpringConstant=2000
    bAngularSpring=true
    bOverrideAngularDamping=true
    bUseGravityOverride=true
    gravityScale=0.35
    simulationSpace=RootRelative
    alphaBoolBlend {blendOption=Linear}
    boundBone {boneName=dyn_coat_fr_l_01}
    componentAppliedLinearAccClamp=<2000 2000 2000>
    componentLinearAccScale=<1 1 1>
    componentLinearVelScale=<1 1 1>
    gravityOverride {z=-980}
    physicsBodyDefinitions [{
      collisionType=CustomSphere
      sphereCollisionRadius=2
      boundBone {boneName=dyn_coat_fr_l_01}
      boxExtents=<20 30 30>
      constraintSetup {angularLimitsMax=<0 0 2>, angularLimitsMin=<0 -30 -2>, angularTarget=<1 0 0>}
      localJointOffset=<2.5 0 0>
     }]
   }
   n31 = Anim Dynamics - Bone: dyn_coat_fr_l_02(ComponentPose=@n30, Alpha=1, GravityOverride="0,0,-1000", ExternalForce="0,0,0") {
    angularDampingOverride=0.9999
    angularSpringConstant=2000
    bAngularSpring=true
    bOverrideAngularDamping=true
    bUseGravityOverride=true
    gravityScale=0.35
    simulationSpace=RootRelative
    alphaBoolBlend {blendOption=Linear}
    boundBone {boneName=dyn_coat_fr_l_02}
    componentAppliedLinearAccClamp=<2000 2000 2000>
    componentLinearAccScale=<1 1 1>
    componentLinearVelScale=<1 1 1>
    gravityOverride {z=-1000}
    physicsBodyDefinitions [{
      collisionType=CustomSphere
      sphereCollisionRadius=2
      boundBone {boneName=dyn_coat_fr_l_02}
      boxExtents=<20 30 30>
      constraintSetup {angularLimitsMax=<0 0 2>, angularLimitsMin=<0 -30 -2>, angularTarget=<1 0 0>}
      localJointOffset=<2.5 0 0>
     }]
   }
   n32 = Anim Dynamics - Bone: dyn_coat_fr_l_03(ComponentPose=@n31, Alpha=1, GravityOverride="0,0,-1000", ExternalForce="0,0,0") {
    angularDampingOverride=0.9999
    angularSpringConstant=2000
    bAngularSpring=true
    bOverrideAngularDamping=true
    bUseGravityOverride=true
    gravityScale=0.35
    simulationSpace=RootRelative
    alphaBoolBlend {blendOption=Linear}
    boundBone {boneName=dyn_coat_fr_l_03}
    componentAppliedLinearAccClamp=<2000 2000 2000>
    componentLinearAccScale=<1 1 1>
    componentLinearVelScale=<1 1 1>
    gravityOverride {z=-1000}
    physicsBodyDefinitions [{
      collisionType=CustomSphere
      sphereCollisionRadius=2
      boundBone {boneName=dyn_coat_fr_l_03}
      boxExtents=<20 30 30>
      constraintSetup {angularLimitsMax=<0 0 2>, angularLimitsMin=<0 -30 -2>, angularTarget=<1 0 0>}
      localJointOffset=<2.5 0 0>
     }]
   }
   n33 = Anim Dynamics - Bone: dyn_coat_side_l_01(ComponentPose=@n32, Alpha=1, GravityOverride="0,0,-980", ExternalForce="0,0,0") {
    angularDampingOverride=0.9999
    angularSpringConstant=2000
    bAngularSpring=true
    bUseGravityOverride=true
    gravityScale=0.35
    simulationSpace=RootRelative
    alphaBoolBlend {blendOption=Linear}
    boundBone {boneName=dyn_coat_side_l_01}
    componentAppliedLinearAccClamp=<2000 2000 2000>
    componentLinearAccScale=<1 1 1>
    componentLinearVelScale=<1 1 1>
    gravityOverride {z=-980}
    physicsBodyDefinitions [{
      collisionType=CustomSphere
      sphereCollisionRadius=2
      boundBone {boneName=dyn_coat_side_l_01}
      boxExtents=<20 30 30>
      constraintSetup {angularLimitsMax=<0 0 2>, angularLimitsMin=<0 -30 -2>, angularTarget=<1 0 0>}
      localJointOffset=<2.5 0 0>
     }]
   }
   n34 = Anim Dynamics - Bone: dyn_coat_side_l_02(ComponentPose=@n33, Alpha=1, GravityOverride="0,0,-1000", ExternalForce="0,0,0") {
    angularDampingOverride=0.9999
    angularSpringConstant=2000
    bAngularSpring=true
    bOverrideAngularDamping=true
    bUseGravityOverride=true
    gravityScale=0.35
    simulationSpace=RootRelative
    alphaBoolBlend {blendOption=Linear}
    boundBone {boneName=dyn_coat_side_l_02}
    componentAppliedLinearAccClamp=<2000 2000 2000>
    componentLinearAccScale=<1 1 1>
    componentLinearVelScale=<1 1 1>
    gravityOverride {z=-1000}
    physicsBodyDefinitions [{
      collisionType=CustomSphere
      sphereCollisionRadius=2
      boundBone {boneName=dyn_coat_side_l_02}
      boxExtents=<20 30 30>
      constraintSetup {angularLimitsMax=<0 0 2>, angularLimitsMin=<0 -30 -2>, angularTarget=<1 0 0>}
      localJointOffset=<2.5 0 0>
     }]
   }
   n35 = Anim Dynamics - Bone: dyn_coat_side_l_03(ComponentPose=@n34, Alpha=1, GravityOverride="0,0,-1000", ExternalForce="0,0,0") {
    angularDampingOverride=0.9999
    angularSpringConstant=2000
    bAngularSpring=true
    bOverrideAngularDamping=true
    bUseGravityOverride=true
    gravityScale=0.35
    simulationSpace=RootRelative
    alphaBoolBlend {blendOption=Linear}
    boundBone {boneName=dyn_coat_side_l_03}
    componentAppliedLinearAccClamp=<2000 2000 2000>
    componentLinearAccScale=<1 1 1>
    componentLinearVelScale=<1 1 1>
    gravityOverride {z=-1000}
    physicsBodyDefinitions [{
      collisionType=CustomSphere
      sphereCollisionRadius=2
      boundBone {boneName=dyn_coat_side_l_03}
      boxExtents=<20 30 30>
      constraintSetup {angularLimitsMax=<0 0 2>, angularLimitsMin=<0 -30 -2>, angularTarget=<1 0 0>}
      localJointOffset=<2.5 0 0>
     }]
   }
   n36 = Anim Dynamics - Bone: dyn_coat_fr_r_01(ComponentPose=@n35, Alpha=1, GravityOverride="0,0,-980", ExternalForce="0,0,0") {
    angularDampingOverride=0.9999
    angularSpringConstant=2000
    bAngularSpring=true
    bUseGravityOverride=true
    gravityScale=0.35
    simulationSpace=RootRelative
    alphaBoolBlend {blendOption=Linear}
    boundBone {boneName=dyn_coat_fr_r_01}
    componentAppliedLinearAccClamp=<2000 2000 2000>
    componentLinearAccScale=<1 1 1>
    componentLinearVelScale=<1 1 1>
    gravityOverride {z=-980}
    physicsBodyDefinitions [{
      collisionType=CustomSphere
      sphereCollisionRadius=2
      boundBone {boneName=dyn_coat_fr_r_01}
      boxExtents=<20 30 30>
      constraintSetup {angularLimitsMax=<0 0 2>, angularLimitsMin=<0 -30 -2>, angularTarget=<1 0 0>}
      localJointOffset=<2.5 0 0>
     }]
   }
   n37 = Anim Dynamics - Bone: dyn_coat_fr_r_02(ComponentPose=@n36, Alpha=1, GravityOverride="0,0,-1000", ExternalForce="0,0,0") {
    angularDampingOverride=0.9999
    angularSpringConstant=2000
    bAngularSpring=true
    bOverrideAngularDamping=true
    bUseGravityOverride=true
    gravityScale=0.35
    simulationSpace=RootRelative
    alphaBoolBlend {blendOption=Linear}
    boundBone {boneName=dyn_coat_fr_r_02}
    componentAppliedLinearAccClamp=<2000 2000 2000>
    componentLinearAccScale=<1 1 1>
    componentLinearVelScale=<1 1 1>
    gravityOverride {z=-1000}
    physicsBodyDefinitions [{
      collisionType=CustomSphere
      sphereCollisionRadius=2
      boundBone {boneName=dyn_coat_fr_r_02}
      boxExtents=<20 30 30>
      constraintSetup {angularLimitsMax=<0 0 2>, angularLimitsMin=<0 -30 -2>, angularTarget=<1 0 0>}
      localJointOffset=<2.5 0 0>
     }]
   }
   n38 = Anim Dynamics - Bone: dyn_coat_fr_r_03(ComponentPose=@n37, Alpha=1, GravityOverride="0,0,-1000", ExternalForce="0,0,0") {
    angularDampingOverride=0.9999
    angularSpringConstant=2000
    bAngularSpring=true
    bOverrideAngularDamping=true
    bUseGravityOverride=true
    gravityScale=0.35
    simulationSpace=RootRelative
    alphaBoolBlend {blendOption=Linear}
    boundBone {boneName=dyn_coat_fr_r_03}
    componentAppliedLinearAccClamp=<2000 2000 2000>
    componentLinearAccScale=<1 1 1>
    componentLinearVelScale=<1 1 1>
    gravityOverride {z=-1000}
    physicsBodyDefinitions [{
      collisionType=CustomSphere
      sphereCollisionRadius=2
      boundBone {boneName=dyn_coat_fr_r_03}
      boxExtents=<20 30 30>
      constraintSetup {angularLimitsMax=<0 0 2>, angularLimitsMin=<0 -30 -2>, angularTarget=<1 0 0>}
      localJointOffset=<2.5 0 0>
     }]
   }
   n39 = Anim Dynamics - Bone: dyn_coat_side_r_01(ComponentPose=@n38, Alpha=1, GravityOverride="0,0,-980", ExternalForce="0,0,0") {
    angularDampingOverride=0.9999
    angularSpringConstant=2000
    bAngularSpring=true
    bUseGravityOverride=true
    gravityScale=0.35
    simulationSpace=RootRelative
    alphaBoolBlend {blendOption=Linear}
    boundBone {boneName=dyn_coat_side_r_01}
    componentAppliedLinearAccClamp=<2000 2000 2000>
    componentLinearAccScale=<1 1 1>
    componentLinearVelScale=<1 1 1>
    gravityOverride {z=-980}
    physicsBodyDefinitions [{
      collisionType=CustomSphere
      sphereCollisionRadius=2
      boundBone {boneName=dyn_coat_side_r_01}
      boxExtents=<20 30 30>
      constraintSetup {angularLimitsMax=<0 0 2>, angularLimitsMin=<0 -30 -2>, angularTarget=<1 0 0>}
      localJointOffset=<2.5 0 0>
     }]
   }
   n40 = Anim Dynamics - Bone: dyn_coat_side_r_02(ComponentPose=@n39, Alpha=1, GravityOverride="0,0,-1000", ExternalForce="0,0,0") {
    angularDampingOverride=0.9999
    angularSpringConstant=2000
    bAngularSpring=true
    bOverrideAngularDamping=true
    bUseGravityOverride=true
    gravityScale=0.35
    simulationSpace=RootRelative
    alphaBoolBlend {blendOption=Linear}
    boundBone {boneName=dyn_coat_side_r_02}
    componentAppliedLinearAccClamp=<2000 2000 2000>
    componentLinearAccScale=<1 1 1>
    componentLinearVelScale=<1 1 1>
    gravityOverride {z=-1000}
    physicsBodyDefinitions [{
      collisionType=CustomSphere
      sphereCollisionRadius=2
      boundBone {boneName=dyn_coat_side_r_02}
      boxExtents=<20 30 30>
      constraintSetup {angularLimitsMax=<0 0 2>, angularLimitsMin=<0 -30 -2>, angularTarget=<1 0 0>}
      localJointOffset=<2.5 0 0>
     }]
   }
   n41 = Anim Dynamics - Bone: dyn_coat_side_r_03(ComponentPose=@n40, Alpha=1, GravityOverride="0,0,-1000", ExternalForce="0,0,0") {
    angularDampingOverride=0.9999
    angularSpringConstant=2000
    bAngularSpring=true
    bOverrideAngularDamping=true
    bUseGravityOverride=true
    gravityScale=0.35
    simulationSpace=RootRelative
    alphaBoolBlend {blendOption=Linear}
    boundBone {boneName=dyn_coat_side_r_03}
    componentAppliedLinearAccClamp=<2000 2000 2000>
    componentLinearAccScale=<1 1 1>
    componentLinearVelScale=<1 1 1>
    gravityOverride {z=-1000}
    physicsBodyDefinitions [{
      collisionType=CustomSphere
      sphereCollisionRadius=2
      boundBone {boneName=dyn_coat_side_r_03}
      boxExtents=<20 30 30>
      constraintSetup {angularLimitsMax=<0 0 2>, angularLimitsMin=<0 -30 -2>, angularTarget=<1 0 0>}
      localJointOffset=<2.5 0 0>
     }]
   }
   n42 = Anim Dynamics - Bone: dyn_coat_bck_l_01(ComponentPose=@n41, Alpha=1, GravityOverride="0,0,-980", ExternalForce="0,0,0") {
    angularDampingOverride=0.9999
    angularSpringConstant=2000
    bAngularSpring=true
    bUseGravityOverride=true
    gravityScale=0.35
    simulationSpace=RootRelative
    alphaBoolBlend {blendOption=Linear}
    boundBone {boneName=dyn_coat_bck_l_01}
    componentAppliedLinearAccClamp=<2000 2000 2000>
    componentLinearAccScale=<1 1 1>
    componentLinearVelScale=<1 1 1>
    gravityOverride {z=-980}
    physicsBodyDefinitions [{
      collisionType=CustomSphere
      sphereCollisionRadius=2
      boundBone {boneName=dyn_coat_bck_l_01}
      boxExtents=<20 30 30>
      constraintSetup {angularLimitsMax=<0 0 2>, angularLimitsMin=<0 -30 -2>, angularTarget=<1 0 0>}
      localJointOffset=<2.5 0 0>
     }]
   }
   n43 = Anim Dynamics - Bone: dyn_coat_bck_l_02(ComponentPose=@n42, Alpha=1, GravityOverride="0,0,-1000", ExternalForce="0,0,0") {
    angularDampingOverride=0.9999
    angularSpringConstant=2000
    bAngularSpring=true
    bOverrideAngularDamping=true
    bUseGravityOverride=true
    gravityScale=0.35
    simulationSpace=RootRelative
    alphaBoolBlend {blendOption=Linear}
    boundBone {boneName=dyn_coat_bck_l_02}
    componentAppliedLinearAccClamp=<2000 2000 2000>
    componentLinearAccScale=<1 1 1>
    componentLinearVelScale=<1 1 1>
    gravityOverride {z=-1000}
    physicsBodyDefinitions [{
      collisionType=CustomSphere
      sphereCollisionRadius=2
      boundBone {boneName=dyn_coat_bck_l_02}
      boxExtents=<20 30 30>
      constraintSetup {angularLimitsMax=<0 0 2>, angularLimitsMin=<0 -30 -2>, angularTarget=<1 0 0>}
      localJointOffset=<2.5 0 0>
     }]
   }
   n44 = Anim Dynamics - Bone: dyn_coat_bck_l_03(ComponentPose=@n43, Alpha=1, GravityOverride="0,0,-1000", ExternalForce="0,0,0") {
    angularDampingOverride=0.9999
    angularSpringConstant=2000
    bAngularSpring=true
    bOverrideAngularDamping=true
    bUseGravityOverride=true
    gravityScale=0.35
    simulationSpace=RootRelative
    alphaBoolBlend {blendOption=Linear}
    boundBone {boneName=dyn_coat_bck_l_03}
    componentAppliedLinearAccClamp=<2000 2000 2000>
    componentLinearAccScale=<1 1 1>
    componentLinearVelScale=<1 1 1>
    gravityOverride {z=-1000}
    physicsBodyDefinitions [{
      collisionType=CustomSphere
      sphereCollisionRadius=2
      boundBone {boneName=dyn_coat_bck_l_03}
      boxExtents=<20 30 30>
      constraintSetup {angularLimitsMax=<0 0 2>, angularLimitsMin=<0 -30 -2>, angularTarget=<1 0 0>}
      localJointOffset=<2.5 0 0>
     }]
   }
   n45 = Anim Dynamics - Bone: dyn_coat_bck_r_01(SimulationSpace=Component, ComponentPose=@n44, Alpha=1, GravityOverride="0,0,-980", ExternalForce="0,0,0") {
    angularDampingOverride=0.9999
    angularSpringConstant=2000
    bAngularSpring=true
    bUseGravityOverride=true
    gravityScale=0.35
    simulationSpace=RootRelative
    alphaBoolBlend {blendOption=Linear}
    boundBone {boneName=dyn_coat_bck_r_01}
    componentAppliedLinearAccClamp=<2000 2000 2000>
    componentLinearAccScale=<1 1 1>
    componentLinearVelScale=<1 1 1>
    gravityOverride {z=-980}
    physicsBodyDefinitions [{
      collisionType=CustomSphere
      sphereCollisionRadius=2
      boundBone {boneName=dyn_coat_bck_r_01}
      boxExtents=<20 30 30>
      constraintSetup {angularLimitsMax=<0 0 2>, angularLimitsMin=<0 -30 -2>, angularTarget=<1 0 0>}
      localJointOffset=<2.5 0 0>
     }]
   }
   n46 = Anim Dynamics - Bone: dyn_coat_bck_r_02(ComponentPose=@n45, Alpha=1, GravityOverride="0,0,-1000", ExternalForce="0,0,0") {
    angularDampingOverride=0.9999
    angularSpringConstant=2000
    bAngularSpring=true
    bOverrideAngularDamping=true
    bUseGravityOverride=true
    gravityScale=0.35
    simulationSpace=RootRelative
    alphaBoolBlend {blendOption=Linear}
    boundBone {boneName=dyn_coat_bck_r_02}
    componentAppliedLinearAccClamp=<2000 2000 2000>
    componentLinearAccScale=<1 1 1>
    componentLinearVelScale=<1 1 1>
    gravityOverride {z=-1000}
    physicsBodyDefinitions [{
      collisionType=CustomSphere
      sphereCollisionRadius=2
      boundBone {boneName=dyn_coat_bck_r_02}
      boxExtents=<20 30 30>
      constraintSetup {angularLimitsMax=<0 0 2>, angularLimitsMin=<0 -30 -2>, angularTarget=<1 0 0>}
      localJointOffset=<2.5 0 0>
     }]
   }
   n47 = Anim Dynamics - Bone: dyn_coat_bck_r_03(ComponentPose=@n46, Alpha=1, GravityOverride="0,0,-1000", ExternalForce="0,0,0") {
    angularDampingOverride=0.9999
    angularSpringConstant=2000
    bAngularSpring=true
    bOverrideAngularDamping=true
    bUseGravityOverride=true
    gravityScale=0.35
    simulationSpace=RootRelative
    alphaBoolBlend {blendOption=Linear}
    boundBone {boneName=dyn_coat_bck_r_03}
    componentAppliedLinearAccClamp=<2000 2000 2000>
    componentLinearAccScale=<1 1 1>
    componentLinearVelScale=<1 1 1>
    gravityOverride {z=-1000}
    physicsBodyDefinitions [{
      collisionType=CustomSphere
      sphereCollisionRadius=2
      boundBone {boneName=dyn_coat_bck_r_03}
      boxExtents=<20 30 30>
      constraintSetup {angularLimitsMax=<0 0 2>, angularLimitsMin=<0 -30 -2>, angularTarget=<1 0 0>}
      localJointOffset=<2.5 0 0>
     }]
   }
   n48 = Component To Local(ComponentPose=@n47)
   n49 = Get bIsPlayingEmote()
   n50 = "Transform (Modify) Bone - Bone: dyn_coat_side_l_01"(
    ComponentPose=Local To Component(LocalPose=Blend Poses by bool(BlendPose_0=@n29, BlendPose_1=@n48, bActiveValue=@n49) {blendType=Linear})
    Alpha=1
    Translation="0,0,0"
    Rotation="0,0,0"
    Scale="1,1,1"
   ) {
    rotationMode=BMM_Additive
    translationMode=BMM_Additive
    alphaBoolBlend {blendOption=Linear}
    alphaScaleBiasClamp {bInterpResult=true}
    boneToModify {boneName=dyn_coat_side_l_01}
   }
   n51 = "Transform (Modify) Bone - Bone: dyn_coat_bck_l_01"(ComponentPose=@n50, Alpha=1, Translation="0,0,0", Rotation="0.000,0.000,-8.091", Scale="1,1,1") {
    rotationMode=BMM_Additive
    translationMode=BMM_Additive
    alphaBoolBlend {blendOption=Linear}
    alphaScaleBiasClamp {bInterpResult=true}
    boneToModify {boneName=dyn_coat_bck_l_01}
    rotation {roll=-8.091}
   }
   n52 = "Transform (Modify) Bone - Bone: dyn_coat_side_l_02"(ComponentPose=@n51, Alpha=1, Translation="0,0,0", Rotation="0,0,0", Scale="1,1,1") {
    rotationMode=BMM_Additive
    translationMode=BMM_Additive
    alphaBoolBlend {blendOption=Linear}
    alphaScaleBiasClamp {bInterpResult=true}
    boneToModify {boneName=dyn_coat_side_l_02}
   }
   n53 = "Transform (Modify) Bone - Bone: dyn_coat_bck_l_02"(ComponentPose=@n52, Alpha=1, Translation="0,0,0", Rotation="0.000,0.000,-14.615", Scale="1,1,1") {
    rotationMode=BMM_Additive
    translationMode=BMM_Additive
    alphaBoolBlend {blendOption=Linear}
    alphaScaleBiasClamp {bInterpResult=true}
    boneToModify {boneName=dyn_coat_bck_l_02}
    rotation {roll=-14.615}
   }
   n54 = "Transform (Modify) Bone - Bone: dyn_coat_bck_r_01"(ComponentPose=@n53, Alpha=1, Translation="0,0,0", Rotation="0.000,0.000,-9.306", Scale="1,1,1") {
    rotationMode=BMM_Additive
    translationMode=BMM_Additive
    alphaBoolBlend {blendOption=Linear}
    alphaScaleBiasClamp {bInterpResult=true}
    boneToModify {boneName=dyn_coat_bck_r_01}
    rotation {roll=-9.306}
   }
   n55 = "Transform (Modify) Bone - Bone: dyn_coat_side_r_01"(ComponentPose=@n54, Alpha=1, Translation="0,0,0", Rotation="0,0,0", Scale="1,1,1") {
    rotationMode=BMM_Additive
    translationMode=BMM_Additive
    alphaBoolBlend {blendOption=Linear}
    alphaScaleBiasClamp {bInterpResult=true}
    boneToModify {boneName=dyn_coat_side_r_01}
   }
   n56 = "Transform (Modify) Bone - Bone: dyn_coat_bck_r_02"(ComponentPose=@n55, Alpha=1, Translation="0,0,0", Rotation="0,0,-8", Scale="1,1,1") {
    rotationMode=BMM_Additive
    translationMode=BMM_Additive
    alphaBoolBlend {blendOption=Linear}
    alphaScaleBiasClamp {bInterpResult=true}
    boneToModify {boneName=dyn_coat_bck_r_02}
    rotation {roll=-8}
   }
   n57 = "Transform (Modify) Bone - Bone: dyn_coat_side_r_02"(ComponentPose=@n56, Alpha=1, Translation="0,0,0", Rotation="0,0,0", Scale="1,1,1") {
    rotationMode=BMM_Additive
    translationMode=BMM_Additive
    alphaBoolBlend {blendOption=Linear}
    alphaScaleBiasClamp {bInterpResult=true}
    boneToModify {boneName=dyn_coat_side_r_02}
   }
   n58 = Anim Dynamics - Bone: dyn_l_flap_01(ComponentPose=@n57, Alpha=1, ExternalForce="0,0,0") {
    angularDampingOverride=1
    angularSpringConstant=8000
    bAngularSpring=true
    bOverrideAngularDamping=true
    bUseGravityOverride=true
    bUsePlanarLimit=false
    bUseSphericalLimits=true
    gravityScale=0.35
    simulationSpace=RootRelative
    alphaBoolBlend {blendOption=Linear}
    boundBone {boneName=dyn_l_flap_01}
    componentAppliedLinearAccClamp=<2000 2000 2000>
    gravityOverride {z=-980}
    physicsBodyDefinitions [{
      boundBone {boneName=dyn_l_flap_01}
      boxExtents=<15 20 20>
      constraintSetup {angularLimitsMax=<0 0 5>, angularLimitsMin=<0 0 -5>, angularTarget=<1 0 0>}
      localJointOffset=<0.5 0 0>
     }]
    sphericalLimits [
     {limitRadius=10, drivingBone {boneName=upperarm_l}, sphereLocalOffset=<4 5 0>}
     {limitRadius=10, drivingBone {boneName=upperarm_twist_01_l}, sphereLocalOffset=<10 2 0>}
    ]
   }
   n59 = Anim Dynamics - Bone: dyn_r_flap_01(ComponentPose=@n58, Alpha=1, ExternalForce="0,0,0") {
    angularDampingOverride=1
    angularSpringConstant=8000
    bAngularSpring=true
    bOverrideAngularDamping=true
    bUseGravityOverride=true
    bUsePlanarLimit=false
    bUseSphericalLimits=true
    gravityScale=0.35
    simulationSpace=RootRelative
    alphaBoolBlend {blendOption=Linear}
    boundBone {boneName=dyn_r_flap_01}
    componentAppliedLinearAccClamp=<2000 2000 2000>
    gravityOverride {z=-980}
    physicsBodyDefinitions [{
      boundBone {boneName=dyn_r_flap_01}
      boxExtents=<15 20 20>
      constraintSetup {angularLimitsMax=<0 0 5>, angularLimitsMin=<0 0 -5>, angularTarget=<1 0 0>}
      localJointOffset=<0.5 0 0>
     }]
   }
   n60 = Anim Dynamics - Bone: dyn_outerCollar_01_l(ComponentPose=@n59, Alpha=1, ExternalForce="0,0,0") {
    angularDampingOverride=1
    angularSpringConstant=8000
    bAngularSpring=true
    bOverrideAngularDamping=true
    bUseGravityOverride=true
    bUseSphericalLimits=true
    gravityScale=0.35
    simulationSpace=RootRelative
    alphaBoolBlend {blendOption=Linear}
    boundBone {boneName=dyn_outerCollar_01_l}
    componentAppliedLinearAccClamp=<2000 2000 2000>
    gravityOverride {z=-980}
    physicsBodyDefinitions [{
      boundBone {boneName=dyn_outerCollar_01_l}
      boxExtents=<15 20 20>
      constraintSetup {angularLimitsMax=<0 5 0>, angularLimitsMin=<0 -5 0>, angularTarget=<1 0 0>}
      localJointOffset=<0.5 0 0>
     }]
    sphericalLimits [{limitRadius=6, drivingBone {boneName=upperarm_l}, sphereLocalOffset=<-2 0 -2>}]
   }
   n61 = Anim Dynamics - Bone: dyn_outerCollar_01_r(ComponentPose=@n60, Alpha=1, ExternalForce="0,0,0") {
    angularDampingOverride=1
    angularSpringConstant=8000
    bAngularSpring=true
    bOverrideAngularDamping=true
    bUseGravityOverride=true
    bUseSphericalLimits=true
    gravityScale=0.35
    simulationSpace=RootRelative
    alphaBoolBlend {blendOption=Linear}
    boundBone {boneName=dyn_outerCollar_01_r}
    componentAppliedLinearAccClamp=<2000 2000 2000>
    gravityOverride {z=-980}
    physicsBodyDefinitions [{
      boundBone {boneName=dyn_outerCollar_01_r}
      boxExtents=<15 20 20>
      constraintSetup {angularLimitsMax=<0 5 0>, angularLimitsMin=<0 -5 0>, angularTarget=<1 0 0>}
      localJointOffset=<0.5 0 0>
     }]
    sphericalLimits [{limitRadius=6, drivingBone {boneName=upperarm_r}, sphereLocalOffset=<2 0 2>}]
   }
   n62 = Get bIsParachuteOpened()
   n63 = "Transform (Modify) Bone - Bone: dyn_outerCollar_01_l"!(ComponentPose=@n61, bAlphaBoolEnabled:bool=@n62, Translation="0,2,0", Rotation="20,0,20", Scale="1,1,1") -> (Pose) {
    alphaInputType=Bool
    rotationMode=BMM_Additive
    translationMode=BMM_Additive
    alphaBoolBlend {blendOption=Linear}
    boneToModify {boneName=dyn_outerCollar_01_l}
    rotation {pitch=20, roll=20}
    translation {y=2}
   }
   n64 = "Transform (Modify) Bone - Bone: dyn_outerCollar_01_r"!(ComponentPose=@n63, bAlphaBoolEnabled:bool=@n62, Translation="0,2,0", Rotation="-20,0,20", Scale="1,1,1") -> (Pose) {
    alphaInputType=Bool
    rotationMode=BMM_Additive
    translationMode=BMM_Additive
    alphaBoolBlend {blendOption=Linear}
    boneToModify {boneName=dyn_outerCollar_01_r}
    rotation {pitch=-20, roll=20}
    translation {y=2}
   }
   n65 = "Transform (Modify) Bone - Bone: dyn_coat_side_l_base"(ComponentPose=@n64, Alpha=1, Translation="0,0,0", Rotation="5,0,0", Scale="1,1,1") {
    rotationMode=BMM_Additive
    translationMode=BMM_Additive
    alphaBoolBlend {blendOption=Linear}
    alphaScaleBiasClamp {bInterpResult=true}
    boneToModify {boneName=dyn_coat_side_l_base}
    rotation {pitch=5}
   }
   n66 = "Transform (Modify) Bone - Bone: dyn_coat_side_r_base"(ComponentPose=@n65, Alpha=1, Translation="0,0,0", Rotation="-5,0,0", Scale="1,1,1") {
    rotationMode=BMM_Additive
    translationMode=BMM_Additive
    alphaBoolBlend {blendOption=Linear}
    alphaScaleBiasClamp {bInterpResult=true}
    boneToModify {boneName=dyn_coat_side_r_base}
    rotation {pitch=-5}
   }
   n67 = RigidBody!(ComponentPose=@n66, Alpha, OverrideWorldGravity:Vector="0,0,-980", ExternalForce) -> (Pose) {
    bClampLinearTranslationLimitToRefPose=true
    bOverrideWorldGravity=true
    overridePhysicsAsset=&/M_MED_LabVapor_CoatLower_SkyDive_Physics
    simulationSpace=BaseBoneSpace
    alphaBoolBlend {blendOption=Linear}
    alphaScaleBiasClamp {interpSpeedDecreasing=15, interpSpeedIncreasing=20}
    baseBoneRef {boneName=pelvis}
    overrideWorldGravity {z=-980}
    simSpaceSettings {maxAngularAcceleration=100, maxAngularVelocity=5, maxLinearAcceleration=1500, maxLinearVelocity=250, velocityScaleZ=0, worldAlpha=1}
   }
   n68 = Anim Dynamics - Bone: dyn_coat_fr_l_01(ComponentPose=@n67, Alpha=1, GravityOverride="0,0,-980", ExternalForce="0,0,0") {
    angularDampingOverride=1
    angularSpringConstant=8000
    bAngularSpring=true
    bOverrideAngularDamping=true
    bUseGravityOverride=true
    gravityScale=0.35
    simulationSpace=RootRelative
    alphaBoolBlend {blendOption=Linear}
    boundBone {boneName=dyn_coat_fr_l_01}
    gravityOverride {z=-980}
    physicsBodyDefinitions [{
      boundBone {boneName=dyn_coat_fr_l_01}
      boxExtents=<20 30 30>
      constraintSetup {angularLimitsMax=<0 0 0>, angularLimitsMin=<0 -20 0>, angularTarget=<1 0 0>}
      localJointOffset=<1 0 0>
     }]
   }
   n69 = Anim Dynamics - Bone: dyn_coat_fr_l_02(ComponentPose=@n68, Alpha=1, GravityOverride="0,0,-980", ExternalForce="0,0,0") {
    angularDampingOverride=1
    angularSpringConstant=8000
    bAngularSpring=true
    bOverrideAngularDamping=true
    bUseGravityOverride=true
    gravityScale=0.35
    simulationSpace=RootRelative
    alphaBoolBlend {blendOption=Linear}
    boundBone {boneName=dyn_coat_fr_l_02}
    gravityOverride {z=-980}
    physicsBodyDefinitions [{
      boundBone {boneName=dyn_coat_fr_l_02}
      boxExtents=<20 30 30>
      constraintSetup {angularLimitsMax=<0 0 0>, angularLimitsMin=<0 -20 0>, angularTarget=<1 0 0>}
      localJointOffset=<1 0 0>
     }]
   }
   n70 = Anim Dynamics - Bone: dyn_coat_fr_l_03(ComponentPose=@n69, Alpha=1, GravityOverride="0,0,-980", ExternalForce="0,0,0") {
    angularDampingOverride=1
    angularSpringConstant=8000
    bAngularSpring=true
    bOverrideAngularDamping=true
    bUseGravityOverride=true
    gravityScale=0.35
    simulationSpace=RootRelative
    alphaBoolBlend {blendOption=Linear}
    boundBone {boneName=dyn_coat_fr_l_03}
    gravityOverride {z=-980}
    physicsBodyDefinitions [{
      boundBone {boneName=dyn_coat_fr_l_03}
      boxExtents=<20 30 30>
      constraintSetup {angularLimitsMax=<0 0 0>, angularLimitsMin=<0 -20 0>, angularTarget=<1 0 0>}
      localJointOffset=<1 0 0>
     }]
   }
   n71 = Anim Dynamics - Bone: dyn_coat_side_l_01(ComponentPose=@n70, Alpha=1, GravityOverride="0,0,-980", ExternalForce="0,0,0") {
    angularDampingOverride=1
    angularSpringConstant=8000
    bAngularSpring=true
    bOverrideAngularDamping=true
    bUseGravityOverride=true
    gravityScale=0.35
    simulationSpace=RootRelative
    alphaBoolBlend {blendOption=Linear}
    boundBone {boneName=dyn_coat_side_l_01}
    gravityOverride {z=-980}
    physicsBodyDefinitions [{
      boundBone {boneName=dyn_coat_side_l_01}
      boxExtents=<20 30 30>
      constraintSetup {angularLimitsMax=<0 0 0>, angularLimitsMin=<0 -20 0>, angularTarget=<1 0 0>}
      localJointOffset=<1 0 0>
     }]
   }
   n72 = Anim Dynamics - Bone: dyn_coat_side_l_02(ComponentPose=@n71, Alpha=1, GravityOverride="0,0,-980", ExternalForce="0,0,0") {
    angularDampingOverride=1
    angularSpringConstant=8000
    bAngularSpring=true
    bOverrideAngularDamping=true
    bUseGravityOverride=true
    gravityScale=0.35
    simulationSpace=RootRelative
    alphaBoolBlend {blendOption=Linear}
    boundBone {boneName=dyn_coat_side_l_02}
    gravityOverride {z=-980}
    physicsBodyDefinitions [{
      boundBone {boneName=dyn_coat_side_l_02}
      boxExtents=<20 30 30>
      constraintSetup {angularLimitsMax=<0 0 0>, angularLimitsMin=<0 -20 0>, angularTarget=<1 0 0>}
      localJointOffset=<1 0 0>
     }]
   }
   n73 = Anim Dynamics - Bone: dyn_coat_side_l_03(ComponentPose=@n72, Alpha=1, GravityOverride="0,0,-980", ExternalForce="0,0,0") {
    angularDampingOverride=1
    angularSpringConstant=8000
    bAngularSpring=true
    bOverrideAngularDamping=true
    bUseGravityOverride=true
    gravityScale=0.35
    simulationSpace=RootRelative
    alphaBoolBlend {blendOption=Linear}
    boundBone {boneName=dyn_coat_side_l_03}
    gravityOverride {z=-980}
    physicsBodyDefinitions [{
      boundBone {boneName=dyn_coat_side_l_03}
      boxExtents=<20 30 30>
      constraintSetup {angularLimitsMax=<0 0 0>, angularLimitsMin=<0 -20 0>, angularTarget=<1 0 0>}
      localJointOffset=<1 0 0>
     }]
   }
   n74 = Anim Dynamics - Bone: dyn_coat_fr_r_01(ComponentPose=@n73, Alpha=1, GravityOverride="0,0,-980", ExternalForce="0,0,0") {
    angularDampingOverride=1
    angularSpringConstant=8000
    bAngularSpring=true
    bOverrideAngularDamping=true
    bUseGravityOverride=true
    gravityScale=0.35
    simulationSpace=RootRelative
    alphaBoolBlend {blendOption=Linear}
    boundBone {boneName=dyn_coat_fr_r_01}
    gravityOverride {z=-980}
    physicsBodyDefinitions [{
      boundBone {boneName=dyn_coat_fr_r_01}
      boxExtents=<20 30 30>
      constraintSetup {angularLimitsMax=<0 0 0>, angularLimitsMin=<0 -20 0>, angularTarget=<1 0 0>}
      localJointOffset=<1 0 0>
     }]
   }
   n75 = Anim Dynamics - Bone: dyn_coat_fr_r_02(ComponentPose=@n74, Alpha=1, GravityOverride="0,0,-980", ExternalForce="0,0,0") {
    angularDampingOverride=1
    angularSpringConstant=8000
    bAngularSpring=true
    bOverrideAngularDamping=true
    bUseGravityOverride=true
    gravityScale=0.35
    simulationSpace=RootRelative
    alphaBoolBlend {blendOption=Linear}
    boundBone {boneName=dyn_coat_fr_r_02}
    gravityOverride {z=-980}
    physicsBodyDefinitions [{
      boundBone {boneName=dyn_coat_fr_r_02}
      boxExtents=<20 30 30>
      constraintSetup {angularLimitsMax=<0 0 0>, angularLimitsMin=<0 -20 0>, angularTarget=<1 0 0>}
      localJointOffset=<1 0 0>
     }]
   }
   n76 = Anim Dynamics - Bone: dyn_coat_fr_r_03(ComponentPose=@n75, Alpha=1, GravityOverride="0,0,-980", ExternalForce="0,0,0") {
    angularDampingOverride=1
    angularSpringConstant=8000
    bAngularSpring=true
    bOverrideAngularDamping=true
    bUseGravityOverride=true
    gravityScale=0.35
    simulationSpace=RootRelative
    alphaBoolBlend {blendOption=Linear}
    boundBone {boneName=dyn_coat_fr_r_03}
    gravityOverride {z=-980}
    physicsBodyDefinitions [{
      boundBone {boneName=dyn_coat_fr_r_03}
      boxExtents=<20 30 30>
      constraintSetup {angularLimitsMax=<0 0 0>, angularLimitsMin=<0 -20 0>, angularTarget=<1 0 0>}
      localJointOffset=<1 0 0>
     }]
   }
   n77 = Anim Dynamics - Bone: dyn_coat_side_r_01(ComponentPose=@n76, Alpha=1, GravityOverride="0,0,-980", ExternalForce="0,0,0") {
    angularDampingOverride=1
    angularSpringConstant=8000
    bAngularSpring=true
    bOverrideAngularDamping=true
    bUseGravityOverride=true
    gravityScale=0.35
    simulationSpace=RootRelative
    alphaBoolBlend {blendOption=Linear}
    boundBone {boneName=dyn_coat_side_r_01}
    gravityOverride {z=-980}
    physicsBodyDefinitions [{
      boundBone {boneName=dyn_coat_side_r_01}
      boxExtents=<20 30 30>
      constraintSetup {angularLimitsMax=<0 0 0>, angularLimitsMin=<0 -20 0>, angularTarget=<1 0 0>}
      localJointOffset=<1 0 0>
     }]
   }
   n78 = Anim Dynamics - Bone: dyn_coat_side_r_02(ComponentPose=@n77, Alpha=1, GravityOverride="0,0,-980", ExternalForce="0,0,0") {
    angularDampingOverride=1
    angularSpringConstant=8000
    bAngularSpring=true
    bOverrideAngularDamping=true
    bUseGravityOverride=true
    gravityScale=0.35
    simulationSpace=RootRelative
    alphaBoolBlend {blendOption=Linear}
    boundBone {boneName=dyn_coat_side_r_02}
    gravityOverride {z=-980}
    physicsBodyDefinitions [{
      boundBone {boneName=dyn_coat_side_r_02}
      boxExtents=<20 30 30>
      constraintSetup {angularLimitsMax=<0 0 0>, angularLimitsMin=<0 -20 0>, angularTarget=<1 0 0>}
      localJointOffset=<1 0 0>
     }]
   }
   n79 = Anim Dynamics - Bone: dyn_coat_side_r_03(ComponentPose=@n78, Alpha=1, GravityOverride="0,0,-980", ExternalForce="0,0,0") {
    angularDampingOverride=1
    angularSpringConstant=8000
    bAngularSpring=true
    bOverrideAngularDamping=true
    bUseGravityOverride=true
    gravityScale=0.35
    simulationSpace=RootRelative
    alphaBoolBlend {blendOption=Linear}
    boundBone {boneName=dyn_coat_side_r_03}
    gravityOverride {z=-980}
    physicsBodyDefinitions [{
      boundBone {boneName=dyn_coat_side_r_03}
      boxExtents=<20 30 30>
      constraintSetup {angularLimitsMax=<0 0 0>, angularLimitsMin=<0 -20 0>, angularTarget=<1 0 0>}
      localJointOffset=<1 0 0>
     }]
   }
   n80 = Anim Dynamics - Bone: dyn_coat_bck_l_01(ComponentPose=@n79, Alpha=1, GravityOverride="0,0,-980", ExternalForce="0,0,0") {
    angularDampingOverride=1
    angularSpringConstant=8000
    bAngularSpring=true
    bOverrideAngularDamping=true
    bUseGravityOverride=true
    gravityScale=0.35
    simulationSpace=RootRelative
    alphaBoolBlend {blendOption=Linear}
    boundBone {boneName=dyn_coat_bck_l_01}
    componentAppliedLinearAccClamp=<100 100 100>
    gravityOverride {z=-980}
    physicsBodyDefinitions [{
      boundBone {boneName=dyn_coat_bck_l_01}
      boxExtents=<20 30 30>
      constraintSetup {angularLimitsMax=<0 0 0>, angularLimitsMin=<0 -20 0>, angularTarget=<1 0 0>}
      localJointOffset=<1 0 0>
     }]
   }
   n81 = Anim Dynamics - Bone: dyn_coat_bck_l_02(ComponentPose=@n80, Alpha=1, GravityOverride="0,0,-980", ExternalForce="0,0,0") {
    angularDampingOverride=1
    angularSpringConstant=8000
    bAngularSpring=true
    bOverrideAngularDamping=true
    bUseGravityOverride=true
    gravityScale=0.35
    simulationSpace=RootRelative
    alphaBoolBlend {blendOption=Linear}
    boundBone {boneName=dyn_coat_bck_l_02}
    componentAppliedLinearAccClamp=<100 100 100>
    gravityOverride {z=-980}
    physicsBodyDefinitions [{
      boundBone {boneName=dyn_coat_bck_l_02}
      boxExtents=<20 30 30>
      constraintSetup {angularLimitsMax=<0 0 0>, angularLimitsMin=<0 -20 0>, angularTarget=<1 0 0>}
      localJointOffset=<1 0 0>
     }]
   }
   n82 = Anim Dynamics - Bone: dyn_coat_bck_l_03(ComponentPose=@n81, Alpha=1, GravityOverride="0,0,-980", ExternalForce="0,0,0") {
    angularDampingOverride=1
    angularSpringConstant=8000
    bAngularSpring=true
    bOverrideAngularDamping=true
    bUseGravityOverride=true
    gravityScale=0.35
    simulationSpace=RootRelative
    alphaBoolBlend {blendOption=Linear}
    boundBone {boneName=dyn_coat_bck_l_03}
    componentAppliedLinearAccClamp=<100 100 100>
    gravityOverride {z=-980}
    physicsBodyDefinitions [{
      boundBone {boneName=dyn_coat_bck_l_03}
      boxExtents=<20 30 30>
      constraintSetup {angularLimitsMax=<0 0 0>, angularLimitsMin=<0 -20 0>, angularTarget=<1 0 0>}
      localJointOffset=<1 0 0>
     }]
   }
   n83 = Anim Dynamics - Bone: dyn_coat_bck_r_01(ComponentPose=@n82, Alpha=1, GravityOverride="0,0,-980", ExternalForce="0,0,0") {
    angularDampingOverride=1
    angularSpringConstant=8000
    bAngularSpring=true
    bOverrideAngularDamping=true
    bUseGravityOverride=true
    gravityScale=0.35
    simulationSpace=RootRelative
    alphaBoolBlend {blendOption=Linear}
    boundBone {boneName=dyn_coat_bck_r_01}
    componentAppliedLinearAccClamp=<100 100 100>
    gravityOverride {z=-980}
    physicsBodyDefinitions [{
      boundBone {boneName=dyn_coat_bck_r_01}
      boxExtents=<20 30 30>
      constraintSetup {angularLimitsMax=<0 0 0>, angularLimitsMin=<0 -20 0>, angularTarget=<1 0 0>}
      localJointOffset=<1 0 0>
     }]
   }
   n84 = Anim Dynamics - Bone: dyn_coat_bck_r_02(ComponentPose=@n83, Alpha=1, GravityOverride="0,0,-980", ExternalForce="0,0,0") {
    angularDampingOverride=1
    angularSpringConstant=8000
    bAngularSpring=true
    bOverrideAngularDamping=true
    bUseGravityOverride=true
    gravityScale=0.35
    simulationSpace=RootRelative
    alphaBoolBlend {blendOption=Linear}
    boundBone {boneName=dyn_coat_bck_r_02}
    componentAppliedLinearAccClamp=<100 100 100>
    gravityOverride {z=-980}
    physicsBodyDefinitions [{
      boundBone {boneName=dyn_coat_bck_r_02}
      boxExtents=<20 30 30>
      constraintSetup {angularLimitsMax=<0 0 0>, angularLimitsMin=<0 -20 0>, angularTarget=<1 0 0>}
      localJointOffset=<1 0 0>
     }]
   }
   n85 = Anim Dynamics - Bone: dyn_coat_bck_r_03(ComponentPose=@n84, Alpha=1, GravityOverride="0,0,-980", ExternalForce="0,0,0") {
    angularDampingOverride=1
    angularSpringConstant=8000
    bAngularSpring=true
    bOverrideAngularDamping=true
    bUseGravityOverride=true
    gravityScale=0.35
    simulationSpace=RootRelative
    alphaBoolBlend {blendOption=Linear}
    boundBone {boneName=dyn_coat_bck_r_03}
    componentAppliedLinearAccClamp=<100 100 100>
    gravityOverride {z=-980}
    physicsBodyDefinitions [{
      boundBone {boneName=dyn_coat_bck_r_03}
      boxExtents=<20 30 30>
      constraintSetup {angularLimitsMax=<0 0 0>, angularLimitsMin=<0 -20 0>, angularTarget=<1 0 0>}
      localJointOffset=<1 0 0>
     }]
   }
   n86 = "Transform (Modify) Bone - Bone: dyn_coat_side_l_02"(ComponentPose=@n85, Alpha=1, Translation="0,0,0", Rotation="0,0,20", Scale="1,1,1") {
    rotationMode=BMM_Additive
    translationMode=BMM_Additive
    alphaBoolBlend {blendOption=Linear}
    alphaScaleBiasClamp {bInterpResult=true}
    boneToModify {boneName=dyn_coat_side_l_02}
    rotation {roll=20}
   }
   n87 = "Transform (Modify) Bone - Bone: dyn_coat_bck_l_02"(ComponentPose=@n86, Alpha=1, Translation="0,0,0", Rotation="0,0,15", Scale="1,1,1") {
    rotationMode=BMM_Additive
    translationMode=BMM_Additive
    alphaBoolBlend {blendOption=Linear}
    alphaScaleBiasClamp {bInterpResult=true}
    boneToModify {boneName=dyn_coat_bck_l_02}
    rotation {roll=15}
   }
   n88 = "Transform (Modify) Bone - Bone: dyn_coat_bck_r_02"(ComponentPose=@n87, Alpha=1, Translation="0,0,0", Rotation="0,0,15", Scale="1,1,1") {
    rotationMode=BMM_Additive
    translationMode=BMM_Additive
    alphaBoolBlend {blendOption=Linear}
    alphaScaleBiasClamp {bInterpResult=true}
    boneToModify {boneName=dyn_coat_bck_r_02}
    rotation {roll=15}
   }
   n89 = "Transform (Modify) Bone - Bone: dyn_coat_side_r_02"(ComponentPose=@n88, Alpha=1, Translation="0,0,0", Rotation="0,0,20", Scale="1,1,1") {
    rotationMode=BMM_Additive
    translationMode=BMM_Additive
    alphaBoolBlend {blendOption=Linear}
    boneToModify {boneName=dyn_coat_side_r_02}
    rotation {roll=20}
   }
   n90 = "Transform (Modify) Bone - Bone: dyn_coat_side_l_02"(ComponentPose=@n89, Alpha=1, Translation="0,0,0", Rotation="0,0,10", Scale="1,1,1") {
    rotationMode=BMM_Additive
    translationMode=BMM_Additive
    alphaBoolBlend {blendOption=Linear}
    boneToModify {boneName=dyn_coat_side_l_02}
    rotation {roll=10}
   }
   n91 = "Transform (Modify) Bone - Bone: dyn_coat_bck_l_02"(ComponentPose=@n90, Alpha=1, Translation="0,0,0", Rotation="0,0,15", Scale="1,1,1") {
    rotationMode=BMM_Additive
    translationMode=BMM_Additive
    alphaBoolBlend {blendOption=Linear}
    boneToModify {boneName=dyn_coat_bck_l_02}
    rotation {roll=15}
   }
   n92 = "Transform (Modify) Bone - Bone: dyn_coat_bck_r_02"(ComponentPose=@n91, Alpha=1, Translation="0,0,0", Rotation="0,0,15", Scale="1,1,1") {
    rotationMode=BMM_Additive
    translationMode=BMM_Additive
    alphaBoolBlend {blendOption=Linear}
    boneToModify {boneName=dyn_coat_bck_r_02}
    rotation {roll=15}
   }
   n93 = "Transform (Modify) Bone - Bone: dyn_coat_side_r_02"(ComponentPose=@n92, Alpha=1, Translation="0,0,0", Rotation="0,0,10", Scale="1,1,1") {
    rotationMode=BMM_Additive
    translationMode=BMM_Additive
    alphaBoolBlend {blendOption=Linear}
    boneToModify {boneName=dyn_coat_side_r_02}
    rotation {roll=10}
   }
   n94 = Control Rig!(Source=Component To Local(ComponentPose=@n93), Alpha, Crouching:bool=Get bIsCrouching()) -> (Pose) {
    alphaBoolBlend {blendOption=Linear}
    controlRigAssetReference {blueprintRigClass=&/M_MED_LabVapor_CtrlRig1.M_MED_LabVapor_CtrlRig1_C}
   }
   n95 = "Transform (Modify) Bone - Bone: dyn_coat_fr_l_base"!(
    ComponentPose=Local To Component(LocalPose=@n94)
    bAlphaBoolEnabled:bool=True
    Translation="0.000,0.945,0.000"
    Rotation="0,0,0"
    Scale="1,1,1"
   ) -> (Pose) {
    alphaInputType=Bool
    rotationMode=BMM_Additive
    translationMode=BMM_Additive
    alphaBoolBlend {blendOption=Linear}
    boneToModify {boneName=dyn_coat_fr_l_base}
    translation {y=0.945}
   }
   n96 = "Transform (Modify) Bone - Bone: dyn_coat_side_l_01"!(ComponentPose=@n95, bAlphaBoolEnabled:bool=True, Translation="-2.467,0.130,0.620", Rotation="0,0,0", Scale="1,1,1") -> (Pose) {
    alphaInputType=Bool
    rotationMode=BMM_Additive
    translationMode=BMM_Additive
    alphaBoolBlend {blendOption=Linear}
    boneToModify {boneName=dyn_coat_side_l_01}
    translation=<-2.467 0.13 0.62>
   }
   n97 = "Transform (Modify) Bone - Bone: dyn_coat_side_l_01"!(ComponentPose=@n96, Alpha:float=1, Translation="-2,0,0", Rotation="0,0,0", Scale="1,1,1") -> (Pose) {
    rotationMode=BMM_Additive
    translationMode=BMM_Additive
    alphaBoolBlend {blendOption=Linear}
    boneToModify {boneName=dyn_coat_side_l_01}
    translation {x=-2}
   }
   n98 = "Transform (Modify) Bone - Bone: dyn_coat_fr_r_02"(ComponentPose=@n97, Alpha=1, Translation="0,0,0", Rotation="0,0,0", Scale="1,1,1") {rotationMode=BMM_Additive, translationMode=BMM_Additive, alphaBoolBlend {blendOption=Linear}, boneToModify {boneName=dyn_coat_fr_r_02}}
   n99 = "Transform (Modify) Bone - Bone: dyn_coat_bck_l_base"(ComponentPose=@n98, Alpha=1, Translation="0,0,0", Rotation="0.000,2.542,-10.000", Scale="1,1,1") {
    rotationMode=BMM_Additive
    translationMode=BMM_Additive
    alphaBoolBlend {blendOption=Linear}
    boneToModify {boneName=dyn_coat_bck_l_base}
    rotation {roll=-10, yaw=2.542}
   }
   n100 = "Transform (Modify) Bone - Bone: dyn_coat_bck_r_base"(ComponentPose=@n99, Alpha=1, Translation="0,0,0", Rotation="0,0,-10", Scale="1,1,1") {
    rotationMode=BMM_Additive
    translationMode=BMM_Additive
    alphaBoolBlend {blendOption=Linear}
    boneToModify {boneName=dyn_coat_bck_r_base}
    rotation {roll=-10}
   }
   n101 = "Transform (Modify) Bone - Bone: dyn_coat_bck_l_02"(ComponentPose=@n100, Alpha=1, Translation="0,0,0", Rotation="0,0,0", Scale="1,1,1") {rotationMode=BMM_Additive, translationMode=BMM_Additive, alphaBoolBlend {blendOption=Linear}, boneToModify {boneName=dyn_coat_bck_l_02}}
   n102 = "Transform (Modify) Bone - Bone: dyn_coat_bck_r_02"(ComponentPose=@n101, Alpha=1, Translation="0,0,0", Rotation="0,0,0", Scale="1,1,1") {rotationMode=BMM_Additive, translationMode=BMM_Additive, alphaBoolBlend {blendOption=Linear}, boneToModify {boneName=dyn_coat_bck_r_02}}
   n103 = "Transform (Modify) Bone - Bone: dyn_coat_fr_r_base"(ComponentPose=@n102, Alpha=1, Translation="-1,1,0", Rotation="0,0,0", Scale="1,1,1") {
    rotationMode=BMM_Additive
    translationMode=BMM_Additive
    alphaBoolBlend {blendOption=Linear}
    boneToModify {boneName=dyn_coat_fr_r_base}
    translation=<-1 1>
   }
   n104 = "Transform (Modify) Bone - Bone: dyn_coat_fr_l_base"!(ComponentPose=@n103, Alpha:float=1, Translation="0,0,0", Rotation="0,0,0", Scale="1,1,1") -> (Pose) {rotationMode=BMM_Additive, translationMode=BMM_Additive, alphaBoolBlend {blendOption=Linear}, boneToModify {boneName=dyn_coat_fr_l_base}}
   n105 = "Transform (Modify) Bone - Bone: dyn_coat_fr_l_base"(ComponentPose=@n104, Alpha=1, Translation="0,0,0", Rotation="0,0,0", Scale="1,1,1") {rotationMode=BMM_Additive, translationMode=BMM_Additive, alphaBoolBlend {blendOption=Linear}, boneToModify {boneName=dyn_coat_fr_l_base}}
   n106 = "Transform (Modify) Bone - Bone: dyn_coat_fr_r_01"(ComponentPose=@n105, Alpha=1, Translation="0,0,0", Rotation="0,0,0", Scale="1,1,1") {rotationMode=BMM_Additive, translationMode=BMM_Additive, alphaBoolBlend {blendOption=Linear}, boneToModify {boneName=dyn_coat_fr_r_01}}
   n107 = "Transform (Modify) Bone - Bone: dyn_coat_bck_r_base"(ComponentPose=@n106, Alpha=1, Translation="0,0,0", Rotation="22.354,36.273,68.016", Scale="1,1,1") {
    rotationMode=BMM_Additive
    translationMode=BMM_Additive
    alphaBoolBlend {blendOption=Linear}
    alphaScaleBiasClamp {bInterpResult=true, interpSpeedDecreasing=5}
    boneToModify {boneName=dyn_coat_bck_r_base}
    rotation=<rot 22.354 36.273 68.016>
   }
   n108 = "Transform (Modify) Bone - Bone: dyn_coat_side_r_base"(ComponentPose=@n107, Alpha=1, Translation="0,0,0", Rotation="-6.409,-7.692,50.431", Scale="1,1,1") {
    rotationMode=BMM_Additive
    translationMode=BMM_Additive
    alphaBoolBlend {blendOption=Linear}
    alphaScaleBiasClamp {bInterpResult=true, interpSpeedDecreasing=5}
    boneToModify {boneName=dyn_coat_side_r_base}
    rotation=<rot -6.409 -7.692 50.431>
   }
   n109 = "Transform (Modify) Bone - Bone: dyn_coat_fr_r_base"(ComponentPose=@n108, Alpha=1, Translation="0,0,0", Rotation="0.000,0.000,30.000", Scale="1,1,1") {
    rotationMode=BMM_Additive
    translationMode=BMM_Additive
    alphaBoolBlend {blendOption=Linear}
    alphaScaleBiasClamp {bInterpResult=true, interpSpeedDecreasing=5}
    boneToModify {boneName=dyn_coat_fr_r_base}
    rotation {roll=30}
   }
   n110 = "Transform (Modify) Bone - Bone: dyn_coat_bck_l_base"(ComponentPose=@n109, Alpha=1, Translation="0,0,0", Rotation="38.737,64.174,68.213", Scale="1,1,1") {
    rotationMode=BMM_Additive
    translationMode=BMM_Additive
    alphaBoolBlend {blendOption=Linear}
    alphaScaleBiasClamp {bInterpResult=true, interpSpeedDecreasing=5}
    boneToModify {boneName=dyn_coat_bck_l_base}
    rotation=<rot 38.737 64.174 68.213>
   }
   n111 = "Transform (Modify) Bone - Bone: dyn_coat_side_l_base"(ComponentPose=@n110, Alpha=1, Translation="0,0,0", Rotation="45.220,44.528,57.892", Scale="1,1,1") {
    rotationMode=BMM_Additive
    translationMode=BMM_Additive
    alphaBoolBlend {blendOption=Linear}
    alphaScaleBiasClamp {bInterpResult=true, interpSpeedDecreasing=5}
    boneToModify {boneName=dyn_coat_side_l_base}
    rotation=<rot 45.22 44.528 57.892>
   }
   n112 = "Transform (Modify) Bone - Bone: dyn_coat_fr_l_base"(ComponentPose=@n111, Alpha=1, Translation="0,0,0", Rotation="9.727,1.308,11.129", Scale="1,1,1") {
    rotationMode=BMM_Additive
    translationMode=BMM_Additive
    alphaBoolBlend {blendOption=Linear}
    alphaScaleBiasClamp {bInterpResult=true, interpSpeedDecreasing=5}
    boneToModify {boneName=dyn_coat_fr_l_base}
    rotation=<rot 9.727 1.308 11.129>
   }
   Output Pose(Result=Component To Local(ComponentPose=@n112)) {layerGroup=None}
   n113 = Control Rig!(
    Source=Copy Pose From Mesh() {bUseAttachedParent=true}
    Alpha
    Crouching=Get crouching() -> (crouching:bool)
    is emote:bool=Get bIsPlayingEmote()
    Celebration Custom:bool=Get Celebration Custom() -> (Celebration Custom:bool)
   ) -> (Pose) {
    alphaBoolBlend {blendOption=Linear}
    controlRigAssetReference {blueprintRigClass=&/M_MED_LabVapor_CtrlRig.M_MED_LabVapor_CtrlRig_C}
   }
   n114 = RigidBody(ComponentPose=Local To Component(LocalPose=@n113), OverrideWorldGravity="0,0,-980") {
    bClampLinearTranslationLimitToRefPose=true
    bOverrideWorldGravity=true
    overridePhysicsAsset=&/M_MED_LabVapor_CoatLower_Physics
    simulationSpace=BaseBoneSpace
    alphaBoolBlend {blendOption=Linear}
    alphaScaleBiasClamp {interpSpeedDecreasing=15, interpSpeedIncreasing=20}
    baseBoneRef {boneName=pelvis}
    overrideWorldGravity {z=-980}
    simSpaceSettings {maxAngularAcceleration=200, maxAngularVelocity=20, maxLinearAcceleration=6000, maxLinearVelocity=1200, velocityScaleZ=0, worldAlpha=1}
   }
   n115 = "Transform (Modify) Bone - Bone: dyn_coat_bck_l_01"!(ComponentPose=@n114, bAlphaBoolEnabled:bool=Get bIsCrouching(), Translation="0,0,-2", Rotation="-5,0,0", Scale="1,1,1") -> (Pose) {
    alphaInputType=Bool
    rotationMode=BMM_Additive
    translationMode=BMM_Additive
    alphaBoolBlend {blendOption=Linear}
    boneToModify {boneName=dyn_coat_bck_l_01}
    rotation {pitch=-5}
    translation {z=-2}
   }
   n116 = "Transform (Modify) Bone - Bone: dyn_coat_bck_l_01"!(ComponentPose=@n115, Alpha:float=1, Translation="0,-1,0", Rotation="0,0,0", Scale="1,1,1") -> (Pose) {
    rotationMode=BMM_Additive
    translationMode=BMM_Additive
    alphaBoolBlend {blendOption=Linear}
    boneToModify {boneName=dyn_coat_bck_l_01}
    translation {y=-1}
   }
   n117 = "Transform (Modify) Bone - Bone: dyn_coat_bck_r_01"(ComponentPose=@n116, Alpha=1, Translation="0,-1,0", Rotation="0,0,0", Scale="1,1,1") {
    rotationMode=BMM_Additive
    translationMode=BMM_Additive
    alphaBoolBlend {blendOption=Linear}
    boneToModify {boneName=dyn_coat_bck_r_01}
    translation {y=-1}
   }
   n118 = "Transform (Modify) Bone - Bone: dyn_coat_side_l_01"(ComponentPose=@n117, Alpha=1, Translation="3,0,0", Rotation="0,0,0", Scale="1,1,1") {
    rotationMode=BMM_Additive
    translationMode=BMM_Additive
    alphaBoolBlend {blendOption=Linear}
    boneToModify {boneName=dyn_coat_side_l_01}
    translation {x=3}
   }
   n119 = "Transform (Modify) Bone - Bone: dyn_coat_fr_l_base"(ComponentPose=@n118, Alpha=Get Crouch Walking() -> (Crouch Walking:double), Translation="1,0,0", Rotation="10,0,0", Scale="1,1,1") {
    rotationMode=BMM_Additive
    translationMode=BMM_Additive
    alphaBoolBlend {blendOption=Linear}
    alphaScaleBiasClamp {bInterpResult=true, interpSpeedIncreasing=6}
    boneToModify {boneName=dyn_coat_fr_l_base}
    rotation {pitch=10}
    translation {x=1}
   }
   n120 = "Transform (Modify) Bone - Bone: dyn_coat_fr_l_01"(ComponentPose=@n119, Alpha=1, Translation="0,0,0", Rotation="-5,0,0", Scale="1,1,1") {
    rotationMode=BMM_Additive
    translationMode=BMM_Additive
    alphaBoolBlend {blendOption=Linear}
    boneToModify {boneName=dyn_coat_fr_l_01}
    rotation {pitch=-5}
   }
   n121 = "Transform (Modify) Bone - Bone: dyn_coat_side_l_base"(ComponentPose=@n120, Alpha=1, Translation="0,0,0", Rotation="-15.137,0.000,-0.000", Scale="1,1,1") {
    rotationMode=BMM_Additive
    translationMode=BMM_Additive
    alphaBoolBlend {blendOption=Linear}
    boneToModify {boneName=dyn_coat_side_l_base}
    rotation {pitch=-15.137}
   }
   Save cached pose 'PreAnimPose'(Pose=Component To Local(ComponentPose=@n121)) {cacheName=PreAnimPose}
   Reroute Node(InputPin:Wildcard) -> (OutputPin:Wildcard)
   Get bIsCrouching()
   Get bIsSkydiveFloating()
  ]}
 EventGraph {nodes [
   n1 = Event BlueprintUpdateAnimation()
   > n2 = Sequence!(execute) -> (then_0, then_1, then_2:Exec, then_3:Exec, then_4:Exec, then_5:Exec, then_6:Exec, then_7:Exec, then_8:Exec, then_9:Exec, then_10:Exec)
   > Branch(Condition=Get bIsInFrontEnd())
   n3 = GetAnimInstance(self=GetOwnerSkeletalMeshComponent())
   > Is Valid(InputObject=@n3)
   n4 = Get bRigidBodyEnabled()
   n5 = Get bIsPlayingEmote()
   > Set engageDuringFrontendIdle(engageDuringFrontendIdle:bool=AND Boolean!(A=Montage_IsPlaying(self=@n3), B=NOT Boolean(A=@n4), C:bool=NOT Boolean(A=@n5)) -> (ReturnValue)) -> (Output_Get:bool)
   n6 = Get bIsKneeSliding()
   n7 = Get bIsBaseballSliding()
   Set coat_dyn_alpha(execute=@n2.then_1, coat_dyn_alpha:double=SelectFloat(A=0.1, B=1, bPickA=OR Boolean(A=@n6, B=@n7))) -> (Output_Get:double)
   > Set twist(twist:float=GetAlphaFromReferenceTransformDelta(
     SkeletalMeshComponent=GetOwnerSkeletalMeshComponent()
     BoneName=head
     BaseName=spine_01
     DeltaComponentType=QuaternionTwist
     InputRange=(X=5,Y=-20)
     TwistAxis="1,0,0"
    )) -> (Output_Get:float)
   > Gravity Override State Function DEPRECATED(CharacterPartAnimInstance=Self-Reference(), GameTimeInSeconds=GetGameTimeInSeconds())
   n8 = Get bIsSliding()
   n9 = Get bIsBaseballSliding()
   n10 = Get bIsKneeSliding()
   Set Sliding(execute=@n2.then_10, Sliding:double="To Float (Boolean)"(InBool=OR Boolean!(A=@n8, B=@n9, C:bool=@n10) -> (ReturnValue))) -> (Output_Get:double)
   n11 = Get bIsSliding()
   n12 = Get bIsBaseballSliding()
   n13 = Get bIsKneeSliding()
   > Set crouching(crouching:bool=OR Boolean!(A=Get bIsCrouching(), B=OR Boolean!(A=@n11, B=@n12, C:bool=@n13) -> (ReturnValue)) -> (ReturnValue)) -> (Output_Get:bool)
   n14 = Get crouching()
   n15 = Get bIsPlayingEmote()
   > Set CrouchandEmotes(CrouchandEmotes:double="To Float (Boolean)"(InBool=OR Boolean(A=@n14, B=@n15))) -> (Output_Get:double)
   Branch(execute=@n2.then_2, Condition=Get bIsInFrontEnd())
   > Set CustomIdleIsPlayingFloat(CustomIdleIsPlayingFloat:float=GetCurveValue(self=GetOwnerFrontEndAnimInstance(), CurveName=Frontend_Idle_Bypass)) -> (Output_Get:float)
   Set simSpace(execute=@n2.then_3, simSpace:AnimPhysSimSpaceType Enum) -> (Output_Get:AnimPhysSimSpaceType Enum)
   > Branch(Condition=OR Boolean!(A=Get bIsDBNO(), B=Get bIsInFrontEnd()) -> (ReturnValue))
   > n16 = Branch(Condition=Get bIsDBNOCarried())
   > Set simSpace(simSpace=World)
   Set simSpace(execute=@n16.else)
   n17 = "Gravity Override State Function (Parameter Asset) (Nativized)"(
    execute=@n2.then_4
    CharacterPartAnimInstance=Self-Reference()
    GameTimeInSeconds=GetGameTimeInSeconds()
    Parameters=&/M_MED_Labvapor_Parameters
   )
   > Set Gravity Override(Gravity Override=@n17)
   > Set SkydiveFloating Alpha(SkydiveFloating Alpha:double="To Float (Boolean)"(InBool=Get bIsSkydiveFloating())) -> (Output_Get:double)
   > Set ParachuteOpenAlpha(ParachuteOpenAlpha:double="To Float (Boolean)"(InBool=Get bIsParachuteOpened())) -> (Output_Get:double)
   n18 = BreakTransform(InTransform=GetDeltaTransformFromRefPose(self=GetOwnerSkeletalMeshComponent(), BoneName=head, BaseName=neck_02))
   Set head_up/down(execute=@n2.then_5, head_up/down:double=MapRangeClamped(Value=BreakRotator(InRot=@n18.Rotation).Yaw, InRangeA=10, InRangeB=-23)) -> (Output_Get:double)
   > Set Crouch Alpha(Crouch Alpha:double="To Float (Boolean)"(InBool=Get bIsCrouching())) -> (Output_Get:double)
   n19 = Get bIsCrouching()
   n20 = Get bIsMoving2D()
   > Set Crouch Walking(Crouch Walking:double="To Float (Boolean)"(InBool=AND Boolean!(A=@n19, B=@n20) -> (ReturnValue))) -> (Output_Get:double)
   n21 = NOT Boolean(A=Get bIsPlayingEmote())
   n22 = Get bIsInFrontEnd()
   > Set FrontEndIdle(FrontEndIdle:double="To Float (Boolean)"(InBool=AND Boolean(A=@n22, B=@n21))) -> (Output_Get:double)
   Branch(execute=@n2.then_6, Condition=Get bIsInFrontEnd())
   > Set FrontEnd Idle Bypass(FrontEnd Idle Bypass:bool=NOT Boolean(A=Get bIsPlayingEmote())) -> (Output_Get:bool)
   > n23 = Branch(Condition=NOT Boolean(A=Get bIsInFrontEnd()))
   > Set FPS(FPS:double=MapRangeClamped(Value=@n1.DeltaTimeX, InRangeA=0.016, InRangeB=0.033)) -> (Output_Get:double)
   Set FPS(execute=@n23.else, FPS=0)
   Set notDiving(execute=@n2.then_8, notDiving:bool=NOT Boolean(A=Get bIsSkyDiving())) -> (Output_Get:bool)
   n24 = GetAnimInstance(self=GetOwnerSkeletalMeshComponent())
   Is Valid(exec=@n2.then_9, InputObject=@n24)
   > Set TonalEmote(TonalEmote:bool=OR Boolean(A=Montage_IsPlaying(self=@n24, Montage=/BRCosmetics/Animation/Game/MainPlayer/Emotes/Tonal/Emote_Tonal_CMM_M), B=false)) -> (Output_Get:bool)
   n25 = GetAnimInstance(self=GetOwnerSkeletalMeshComponent())
   Is Valid(exec=@n2.then_7, InputObject=@n25)
   > Set Celebration Custom(Celebration Custom:bool=OR Boolean(A=Montage_IsPlaying(self=@n25, Montage=/Game/Animation/Game/MainPlayer/Emotes/Emote_Celebration_M), B=false)) -> (Output_Get:bool)
   Set frontendalpha(frontendalpha:double=SelectFloat(A=1, B=0, bPickA=Get bIsInFrontEnd())) -> (Output_Get:double)
   n26 = Get bIsPlayingEmote()
   > Set FrontEnd Idle Bypass_0(FrontEnd Idle Bypass_0:bool=AND Boolean(A=NOT Boolean(A=@n26), B=Get bIsInFrontEnd())) -> (Output_Get:bool)
   AND Boolean(A=false, B=false)
   "float > float"(A=GetCurveValue(self=GetOwnerFrontEndAnimInstance(), CurveName=FrontEnd_Idle_Bypass), B=0.1)
  ]}
]
variables [
 Celebration Custom {type=boolean}
 coat_dyn_alpha {type=number, default=1}
 Crouch Alpha {type=number}
 Crouch Walking {type=number}
 CrouchandEmotes {type=number}
 crouching {type=boolean}
 CustomIdleIsPlayingFloat {type=number}
 engageDuringFrontendIdle {type=boolean}
 FPS {type=number}
 FrontEnd Idle Bypass {type=boolean}
 FrontEnd Idle Bypass_0 {type=boolean}
 frontendalpha {type=number}
 FrontEndIdle {type=number}
 Gravity Override {type=Vector}
 Gravity Override Params {type=GravityOverrideParamsStruct}
 head_up/down {type=number, default=0.303}
 notDiving {type=boolean}
 ParachuteOpenAlpha {type=number}
 simSpace {type=AnimPhysSimSpaceType, default=RootRelative}
 SkyDiveFloat_Alpha {type=boolean}
 SkydiveFloating Alpha {type=number}
 Sliding {type=number}
 TonalEmote {type=boolean}
 twist {type=number}
]
