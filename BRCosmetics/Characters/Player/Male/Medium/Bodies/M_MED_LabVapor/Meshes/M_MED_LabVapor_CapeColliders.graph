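# Physics asset of collision shapes for the LabVapor medium male player body, named as cape colliders.
# Capsules and a head sphere cover torso, neck and clavicles, while arm and leg bodies have no shapes.
# Six coat chains (front, side, back per side) hang from the pelvis, each with four capsule bodies.
# A huge zero-length root capsule approximates a floor at the feet, and all constraints leave swing and twist free.
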PhysicsAsset /BRCosmetics/Characters/Player/Male/Medium/Bodies/M_MED_LabVapor/Meshes/M_MED_LabVapor_CapeColliders
bodies {
 calf_l []
 calf_r []
 clavicle_l [{shapeType=Capsule, length=9.0192, radius=9.8871, center=<9.4266 -1.1806 -2.4977>, rotation=<rot 69.3297 1.4476 176.9462>}]
 clavicle_r [{shapeType=Capsule, length=9.0192, radius=9.8871, center=<-9.4266 1.1806 2.4977>, rotation=<rot 69.3297 1.4476 -3.0538>}]
 dyn_coat_bck_l_01 [{shapeType=Capsule, length=10, radius=7, shapeName=dyn_coat_bck_l_01_capsule, center=<7 0 -2>, rotation=<rot 0 0 90>}]
 dyn_coat_bck_l_02 [{shapeType=Capsule, length=10, radius=8, shapeName=dyn_coat_bck_l_02_capsule, center=<7.3545 0 2>, rotation=<rot 0 0 90>}]
 dyn_coat_bck_l_03 [{shapeType=Capsule, length=15, radius=8, shapeName=dyn_coat_bck_l_03_capsule, center=<4.1178 0 2.8105>, rotation=<rot 0 0 90>}]
 dyn_coat_bck_l_base [{shapeType=Capsule, length=10, radius=7, shapeName=dyn_coat_bck_l_base_capsule, center=<4 0 -2>, rotation=<rot 0 0 90>}]
 dyn_coat_bck_r_01 [{shapeType=Capsule, length=10, radius=7, shapeName=dyn_coat_bck_r_01_capsule, center=<7 0 -2>, rotation=<rot 0 0 90>}]
 dyn_coat_bck_r_02 [{shapeType=Capsule, length=10, radius=8, shapeName=dyn_coat_bck_r_02_capsule, center=<7.3545 0 2>, rotation=<rot 0 0 90>}]
 dyn_coat_bck_r_03 [{shapeType=Capsule, length=15, radius=8, shapeName=dyn_coat_bck_r_03_capsule, center=<4.1178 0 1.0852>, rotation=<rot 0 0 90>}]
 dyn_coat_bck_r_base [{shapeType=Capsule, length=10, radius=7, shapeName=dyn_coat_bck_r_base_capsule, center=<4 0 -2>, rotation=<rot 0 0 90>}]
 dyn_coat_fr_l_01 [{shapeType=Capsule, length=10, radius=7, shapeName=dyn_coat_fr_l_01_capsule, center=<7 0 -2>, rotation=<rot 0 0 90>}]
 dyn_coat_fr_l_02 [{shapeType=Capsule, length=10, radius=8, shapeName=dyn_coat_fr_l_02_capsule, center=<7.3545 0 2>, rotation=<rot 0 0 90>}]
 dyn_coat_fr_l_03 [{shapeType=Capsule, length=15, radius=8, shapeName=dyn_coat_fr_l_03_capsule, center=<4.1178 0 -2>, rotation=<rot 0 0 90>}]
 dyn_coat_fr_l_base [{shapeType=Capsule, length=10, radius=7, shapeName=dyn_coat_fr_l_base_capsule, center=<4 0 -2>, rotation=<rot 0 0 90>}]
 dyn_coat_fr_r_01 [{shapeType=Capsule, length=10, radius=7, shapeName=dyn_coat_fr_r_01_capsule, center=<7 0 -2>, rotation=<rot 0 0 90>}]
 dyn_coat_fr_r_02 [{shapeType=Capsule, length=10, radius=8, shapeName=dyn_coat_fr_r_02_capsule, center=<7.3545 0 2>, rotation=<rot 0 0 90>}]
 dyn_coat_fr_r_03 [{shapeType=Capsule, length=15, radius=8, shapeName=dyn_coat_fr_r_03_capsule, center=<4.1178 0 -2>, rotation=<rot 0 0 90>}]
 dyn_coat_fr_r_base [{shapeType=Capsule, length=10, radius=7, shapeName=dyn_coat_fr_r_base_capsule, center=<4 0 -2>, rotation=<rot 0 0 90>}]
 dyn_coat_side_l_01 [{shapeType=Capsule, length=10, radius=7, shapeName=dyn_coat_side_l_01_capsule, center=<7 0 -2>, rotation=<rot 0 0 90>}]
 dyn_coat_side_l_02 [{shapeType=Capsule, length=10, radius=8, shapeName=dyn_coat_side_l_02_capsule, center=<7.3545 0 2>, rotation=<rot 0 0 90>}]
 dyn_coat_side_l_03 [{shapeType=Capsule, length=15, radius=8, shapeName=dyn_coat_side_l_03_capsule, center=<4.1178 0 0.9139>, rotation=<rot 0 0 90>}]
 dyn_coat_side_l_base [{shapeType=Capsule, length=10, radius=7, shapeName=dyn_coat_side_l_base_capsule, center=<4 0 -2>, rotation=<rot 0 0 90>}]
 dyn_coat_side_r_01 [{shapeType=Capsule, length=10, radius=7, shapeName=dyn_coat_side_r_01_capsule, center=<7 0 -2>, rotation=<rot 0 0 90>}]
 dyn_coat_side_r_02 [{shapeType=Capsule, length=10, radius=8, shapeName=dyn_coat_side_r_02_capsule, center=<7.3545 0 2>, rotation=<rot 0 0 90>}]
 dyn_coat_side_r_03 [{shapeType=Capsule, length=15, radius=8, shapeName=dyn_coat_side_r_03_capsule, center=<4.1178 0 0.9139>, rotation=<rot 0 0 90>}]
 dyn_coat_side_r_base [{shapeType=Capsule, length=10, radius=7, shapeName=dyn_coat_side_r_base_capsule, center=<4 0 -2>, rotation=<rot 0 0 90>}]
 foot_l []
 foot_r []
 head [{shapeType=Sphere, radius=18.2413, center=<9.2123 -2.2909 0>}]
 lowerarm_l []
 lowerarm_r []
 neck_02 [{shapeType=Capsule, length=13.4375, radius=7.3175, center=<-0.2642 -0.3408 0>, rotation=<rot 90 -171.1193 -175.3978>}]
 pelvis [{shapeType=Capsule, length=24.5919, radius=16.4861, center=<-2.4887 0.1881 0>}]
 root [{shapeType=Capsule, length=0, radius=289.5, shapeName=root_capsule, center=<0 0 -286.8952>}]
 spine_01 [{shapeType=Capsule, length=25.4313, radius=16.0094, center=<-1.7965 0.1884 0>}]
 spine_02 [
  {shapeType=Capsule, length=25.0655, radius=13.6322, center=<-1.711 -0.1851 0.0001>}
  {shapeType=Capsule, length=29.3956, radius=11.5807, center=<2.5024 0.4992 3.9825>, rotation=<rot 90 0.0049 0.0049>}
  {shapeType=Capsule, length=29.875, radius=11.1013, center=<2.4704 0.4991 -3.9825>, rotation=<rot 90 -0.0396 -0.0396>}
 ]
 spine_03 [{shapeType=Capsule, length=22.6162, radius=14.2578, center=<0.1445 0.2396 0.0001>}]
 spine_04 [{shapeType=Capsule, length=25.2154, radius=13.9444, center=<2.766 0.7765 0.0002>}]
 spine_05 [
  {shapeType=Capsule, length=10.356, radius=13.3616, center=<9.4958 0.4385 0>}
  {shapeType=Capsule, length=14.7, radius=13.1943, center=<-0.2916 0.4452 0>}
 ]
 thigh_l []
 thigh_r []
 upperarm_l []
 upperarm_r []
}
constraints [
 {
  bone1Name=spine_01
  bone2Name=pelvis
  swing1LimitDegrees=45
  swing1Motion=Free
  swing2LimitDegrees=45
  swing2Motion=Free
  twistLimitDegrees=45
  twistMotion=Free
 }
 {
  bone1Name=spine_02
  bone2Name=spine_01
  swing1LimitDegrees=45
  swing1Motion=Free
  swing2LimitDegrees=45
  swing2Motion=Free
  twistLimitDegrees=45
  twistMotion=Free
 }
 {
  bone1Name=spine_03
  bone2Name=spine_02
  swing1LimitDegrees=45
  swing1Motion=Free
  swing2LimitDegrees=45
  swing2Motion=Free
  twistLimitDegrees=45
  twistMotion=Free
 }
 {
  bone1Name=spine_04
  bone2Name=spine_03
  swing1LimitDegrees=45
  swing1Motion=Free
  swing2LimitDegrees=45
  swing2Motion=Free
  twistLimitDegrees=45
  twistMotion=Free
 }
 {
  bone1Name=spine_05
  bone2Name=spine_04
  swing1LimitDegrees=45
  swing1Motion=Free
  swing2LimitDegrees=45
  swing2Motion=Free
  twistLimitDegrees=45
  twistMotion=Free
 }
 {
  bone1Name=neck_02
  bone2Name=spine_05
  swing1LimitDegrees=45
  swing1Motion=Free
  swing2LimitDegrees=45
  swing2Motion=Free
  twistLimitDegrees=45
  twistMotion=Free
 }
 {
  bone1Name=head
  bone2Name=neck_02
  swing1LimitDegrees=45
  swing1Motion=Free
  swing2LimitDegrees=45
  swing2Motion=Free
  twistLimitDegrees=45
  twistMotion=Free
 }
 {
  bone1Name=thigh_l
  bone2Name=pelvis
  swing1LimitDegrees=45
  swing1Motion=Free
  swing2LimitDegrees=45
  swing2Motion=Free
  twistLimitDegrees=45
  twistMotion=Free
 }
 {
  bone1Name=foot_l
  bone2Name=calf_l
  swing1LimitDegrees=45
  swing1Motion=Free
  swing2LimitDegrees=45
  swing2Motion=Free
  twistLimitDegrees=45
  twistMotion=Free
 }
 {
  bone1Name=clavicle_r
  bone2Name=spine_05
  swing1LimitDegrees=45
  swing1Motion=Free
  swing2LimitDegrees=45
  swing2Motion=Free
  twistLimitDegrees=45
  twistMotion=Free
 }
 {
  bone1Name=upperarm_r
  bone2Name=clavicle_r
  swing1LimitDegrees=45
  swing1Motion=Free
  swing2LimitDegrees=45
  swing2Motion=Free
  twistLimitDegrees=45
  twistMotion=Free
 }
 {
  bone1Name=lowerarm_r
  bone2Name=upperarm_r
  swing1LimitDegrees=45
  swing1Motion=Free
  swing2LimitDegrees=45
  swing2Motion=Free
  twistLimitDegrees=45
  twistMotion=Free
 }
 {
  bone1Name=foot_r
  bone2Name=calf_r
  swing1LimitDegrees=45
  swing1Motion=Free
  swing2LimitDegrees=45
  swing2Motion=Free
  twistLimitDegrees=45
  twistMotion=Free
 }
 {
  bone1Name=lowerarm_l
  bone2Name=upperarm_l
  swing1LimitDegrees=45
  swing1Motion=Free
  swing2LimitDegrees=45
  swing2Motion=Free
  twistLimitDegrees=45
  twistMotion=Free
 }
 {
  bone1Name=upperarm_l
  bone2Name=clavicle_l
  swing1LimitDegrees=45
  swing1Motion=Free
  swing2LimitDegrees=45
  swing2Motion=Free
  twistLimitDegrees=45
  twistMotion=Free
 }
 {
  bone1Name=clavicle_l
  bone2Name=spine_05
  swing1LimitDegrees=45
  swing1Motion=Free
  swing2LimitDegrees=45
  swing2Motion=Free
  twistLimitDegrees=45
  twistMotion=Free
 }
 {
  bone1Name=calf_r
  bone2Name=thigh_r
  swing1LimitDegrees=45
  swing1Motion=Free
  swing2LimitDegrees=45
  swing2Motion=Free
  twistLimitDegrees=45
  twistMotion=Free
 }
 {
  bone1Name=thigh_r
  bone2Name=pelvis
  swing1LimitDegrees=45
  swing1Motion=Free
  swing2LimitDegrees=45
  swing2Motion=Free
  twistLimitDegrees=45
  twistMotion=Free
 }
 {
  bone1Name=calf_l
  bone2Name=thigh_l
  swing1LimitDegrees=45
  swing1Motion=Free
  swing2LimitDegrees=45
  swing2Motion=Free
  twistLimitDegrees=45
  twistMotion=Free
 }
 {
  bone1Name=dyn_coat_bck_l_base
  bone2Name=pelvis
  swing1LimitDegrees=45
  swing1Motion=Free
  swing2LimitDegrees=45
  swing2Motion=Free
  twistLimitDegrees=45
  twistMotion=Free
 }
 {
  bone1Name=dyn_coat_bck_l_01
  bone2Name=dyn_coat_bck_l_base
  swing1LimitDegrees=45
  swing1Motion=Free
  swing2LimitDegrees=45
  swing2Motion=Free
  twistLimitDegrees=45
  twistMotion=Free
 }
 {
  bone1Name=dyn_coat_bck_l_02
  bone2Name=dyn_coat_bck_l_01
  swing1LimitDegrees=45
  swing1Motion=Free
  swing2LimitDegrees=45
  swing2Motion=Free
  twistLimitDegrees=45
  twistMotion=Free
 }
 {
  bone1Name=dyn_coat_bck_l_03
  bone2Name=dyn_coat_bck_l_02
  swing1LimitDegrees=45
  swing1Motion=Free
  swing2LimitDegrees=45
  swing2Motion=Free
  twistLimitDegrees=45
  twistMotion=Free
 }
 {
  bone1Name=dyn_coat_bck_r_base
  bone2Name=pelvis
  swing1LimitDegrees=45
  swing1Motion=Free
  swing2LimitDegrees=45
  swing2Motion=Free
  twistLimitDegrees=45
  twistMotion=Free
 }
 {
  bone1Name=dyn_coat_bck_r_01
  bone2Name=dyn_coat_bck_r_base
  swing1LimitDegrees=45
  swing1Motion=Free
  swing2LimitDegrees=45
  swing2Motion=Free
  twistLimitDegrees=45
  twistMotion=Free
 }
 {
  bone1Name=dyn_coat_bck_r_02
  bone2Name=dyn_coat_bck_r_01
  swing1LimitDegrees=45
  swing1Motion=Free
  swing2LimitDegrees=45
  swing2Motion=Free
  twistLimitDegrees=45
  twistMotion=Free
 }
 {
  bone1Name=dyn_coat_bck_r_03
  bone2Name=dyn_coat_bck_r_02
  swing1LimitDegrees=45
  swing1Motion=Free
  swing2LimitDegrees=45
  swing2Motion=Free
  twistLimitDegrees=45
  twistMotion=Free
 }
 {
  bone1Name=dyn_coat_fr_l_base
  bone2Name=pelvis
  swing1LimitDegrees=45
  swing1Motion=Free
  swing2LimitDegrees=45
  swing2Motion=Free
  twistLimitDegrees=45
  twistMotion=Free
 }
 {
  bone1Name=dyn_coat_fr_l_01
  bone2Name=dyn_coat_fr_l_base
  swing1LimitDegrees=45
  swing1Motion=Free
  swing2LimitDegrees=45
  swing2Motion=Free
  twistLimitDegrees=45
  twistMotion=Free
 }
 {
  bone1Name=dyn_coat_fr_l_02
  bone2Name=dyn_coat_fr_l_01
  swing1LimitDegrees=45
  swing1Motion=Free
  swing2LimitDegrees=45
  swing2Motion=Free
  twistLimitDegrees=45
  twistMotion=Free
 }
 {
  bone1Name=dyn_coat_fr_l_03
  bone2Name=dyn_coat_fr_l_02
  swing1LimitDegrees=45
  swing1Motion=Free
  swing2LimitDegrees=45
  swing2Motion=Free
  twistLimitDegrees=45
  twistMotion=Free
 }
 {
  bone1Name=dyn_coat_fr_r_base
  bone2Name=pelvis
  swing1LimitDegrees=45
  swing1Motion=Free
  swing2LimitDegrees=45
  swing2Motion=Free
  twistLimitDegrees=45
  twistMotion=Free
 }
 {
  bone1Name=dyn_coat_fr_r_01
  bone2Name=dyn_coat_fr_r_base
  swing1LimitDegrees=45
  swing1Motion=Free
  swing2LimitDegrees=45
  swing2Motion=Free
  twistLimitDegrees=45
  twistMotion=Free
 }
 {
  bone1Name=dyn_coat_fr_r_02
  bone2Name=dyn_coat_fr_r_01
  swing1LimitDegrees=45
  swing1Motion=Free
  swing2LimitDegrees=45
  swing2Motion=Free
  twistLimitDegrees=45
  twistMotion=Free
 }
 {
  bone1Name=dyn_coat_fr_r_03
  bone2Name=dyn_coat_fr_r_02
  swing1LimitDegrees=45
  swing1Motion=Free
  swing2LimitDegrees=45
  swing2Motion=Free
  twistLimitDegrees=45
  twistMotion=Free
 }
 {
  bone1Name=dyn_coat_side_l_base
  bone2Name=pelvis
  swing1LimitDegrees=45
  swing1Motion=Free
  swing2LimitDegrees=45
  swing2Motion=Free
  twistLimitDegrees=45
  twistMotion=Free
 }
 {
  bone1Name=dyn_coat_side_l_01
  bone2Name=dyn_coat_side_l_base
  swing1LimitDegrees=45
  swing1Motion=Free
  swing2LimitDegrees=45
  swing2Motion=Free
  twistLimitDegrees=45
  twistMotion=Free
 }
 {
  bone1Name=dyn_coat_side_l_02
  bone2Name=dyn_coat_side_l_01
  swing1LimitDegrees=45
  swing1Motion=Free
  swing2LimitDegrees=45
  swing2Motion=Free
  twistLimitDegrees=45
  twistMotion=Free
 }
 {
  bone1Name=dyn_coat_side_l_03
  bone2Name=dyn_coat_side_l_02
  swing1LimitDegrees=45
  swing1Motion=Free
  swing2LimitDegrees=45
  swing2Motion=Free
  twistLimitDegrees=45
  twistMotion=Free
 }
 {
  bone1Name=dyn_coat_side_r_base
  bone2Name=pelvis
  swing1LimitDegrees=45
  swing1Motion=Free
  swing2LimitDegrees=45
  swing2Motion=Free
  twistLimitDegrees=45
  twistMotion=Free
 }
 {
  bone1Name=dyn_coat_side_r_01
  bone2Name=dyn_coat_side_r_base
  swing1LimitDegrees=45
  swing1Motion=Free
  swing2LimitDegrees=45
  swing2Motion=Free
  twistLimitDegrees=45
  twistMotion=Free
 }
 {
  bone1Name=dyn_coat_side_r_02
  bone2Name=dyn_coat_side_r_01
  swing1LimitDegrees=45
  swing1Motion=Free
  swing2LimitDegrees=45
  swing2Motion=Free
  twistLimitDegrees=45
  twistMotion=Free
 }
 {
  bone1Name=dyn_coat_side_r_03
  bone2Name=dyn_coat_side_r_02
  swing1LimitDegrees=45
  swing1Motion=Free
  swing2LimitDegrees=45
  swing2Motion=Free
  twistLimitDegrees=45
  twistMotion=Free
 }
 {
  bone1Name=pelvis
  bone2Name=root
  swing1LimitDegrees=45
  swing1Motion=Free
  swing2LimitDegrees=45
  swing2Motion=Free
  twistLimitDegrees=45
  twistMotion=Free
 }
]
