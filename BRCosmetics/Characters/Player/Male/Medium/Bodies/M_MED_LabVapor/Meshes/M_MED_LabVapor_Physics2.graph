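# Physics asset for the male LabVapor body, setting up capsule collision and joints for a dynamic coat over the pelvis and legs.
# Six coat chains (front, side and back per side) hang from the pelvis, each a base plus three segments with narrow swing limits and zero twist.
# Extra angularly free constraints tie neighboring lower panels together, asymmetrically between left and right.
# Pelvis, calves and feet carry capsules; thigh bodies have no shapes.
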PhysicsAsset /BRCosmetics/Characters/Player/Male/Medium/Bodies/M_MED_LabVapor/Meshes/M_MED_LabVapor_Physics2
bodies {
 calf_l [{shapeType=Capsule, length=55.529, radius=5, center=<-25.8127 -0.5835 1.2082>, rotation=<rot -3.3016 86.6217 -92.3669>}]
 calf_r [{shapeType=Capsule, length=55.529, radius=5, center=<25.8127 0.5835 -1.2082>, rotation=<rot -3.3016 86.6217 87.6331>}]
 dyn_L_backCoat_1 [{shapeType=Capsule, length=17, radius=6.5, shapeName=dyn_L_backCoat_1_capsule, center=<6.0913 0 3.25>, rotation=<rot -0.0002 0 95>}]
 dyn_L_backCoat_2 [{shapeType=Capsule, length=15, radius=6, shapeName=dyn_L_backCoat_2_capsule, center=<4.8742 0 0.5>, rotation=<rot -0.0003 0 100>}]
 dyn_L_backCoat_3 [{shapeType=Capsule, length=18, radius=7, shapeName=dyn_L_backCoat_3_capsule, center=<-1.3355 0 1.5>, rotation=<rot -0.0003 0 100>}]
 dyn_L_backCoat_base [{shapeType=Capsule, length=1, radius=1, shapeName=dyn_L_backCoat_base_capsule, rotation=<rot 0 0 90>}]
 dyn_L_frontCoat_1 [{shapeType=Capsule, length=19, radius=7, shapeName=dyn_L_frontCoat_1_capsule, center=<5.4339 0 6.5>, rotation=<rot -2.1108 4.5342 -70.0837>}]
 dyn_L_frontCoat_2 [{shapeType=Capsule, length=14.5, radius=8, shapeName=dyn_L_frontCoat_2_capsule, center=<1.8117 0 7>, rotation=<rot -1.2924 4.8313 -70.0547>}]
 dyn_L_frontCoat_3 [{
   shapeType=Capsule
   length=14.288
   radius=7
   shapeName=dyn_L_frontCoat_3_capsule
   center=<4.9353 4.0688 15.9954>
   rotation=<rot -5.2229 4.1109 -70.0533>
  }]
 dyn_L_frontCoat_base [{shapeType=Capsule, length=2, radius=1, shapeName=dyn_L_frontCoat_base_capsule, rotation=<rot 0 0 -75>}]
 dyn_L_sideCoat_1 [{shapeType=Capsule, length=9, radius=7, shapeName=dyn_L_sideCoat_1_capsule, center=<3.7813 0 4.5>, rotation=<rot 85 89.9987 -0.0015>}]
 dyn_L_sideCoat_2 [{shapeType=Capsule, length=11.5, radius=7, shapeName=dyn_L_sideCoat_1_capsule, center=<1.2082 0 7>, rotation=<rot 85 89.9987 -0.0015>}]
 dyn_L_sideCoat_3 [{shapeType=Capsule, length=18.875, radius=7, shapeName=dyn_L_sideCoat_1_capsule, center=<1.1845 0 5>, rotation=<rot 85 89.9987 -0.0015>}]
 dyn_L_sideCoat_base [{shapeType=Capsule, length=1, radius=1, shapeName=dyn_L_sideCoat_base_capsule, rotation=<rot 90 90 0>}]
 dyn_R_backCoat_1 [{shapeType=Capsule, length=17, radius=6.5, shapeName=dyn_L_backCoat_1_capsule, center=<6.0913 0 3.25>, rotation=<rot -0.0002 0 95>}]
 dyn_R_backCoat_2 [{shapeType=Capsule, length=15, radius=6, shapeName=dyn_L_backCoat_2_capsule, center=<4.8742 0 0.5>, rotation=<rot -0.0004 0 100>}]
 dyn_R_backCoat_3 [{shapeType=Capsule, length=18, radius=7, shapeName=dyn_L_backCoat_3_capsule, center=<-1.3355 0 1.5>, rotation=<rot -0.0003 0 100>}]
 dyn_R_backCoat_base [{shapeType=Capsule, length=1, radius=1, shapeName=dyn_R_backCoat_base_capsule, rotation=<rot 0 0 -90>}]
 dyn_R_frontCoat_1 [{
   shapeType=Capsule
   length=18.875
   radius=7
   shapeName=dyn_L_frontCoat_1_capsule
   center=<5.4339 0 6.5>
   rotation=<rot -4.6707 -5.1508 -104.5093>
  }]
 dyn_R_frontCoat_2 [{shapeType=Capsule, length=17, radius=8, shapeName=dyn_L_frontCoat_2_capsule, center=<1.8015 0 9>, rotation=<rot -1.282 -5.1724 -104.8292>}]
 dyn_R_frontCoat_3 [{
   shapeType=Capsule
   length=18.788
   radius=7
   shapeName=dyn_L_frontCoat_3_capsule
   center=<3.6267 0 5>
   rotation=<rot -5.2227 -4.1117 -105.0531>
  }]
 dyn_R_frontCoat_base [{shapeType=Capsule, length=1, radius=1, shapeName=dyn_R_frontCoat_base_capsule, rotation=<rot 90 90 0>}]
 dyn_R_sideCoat_1 [{shapeType=Capsule, length=9, radius=7, shapeName=dyn_L_sideCoat_1_capsule, center=<3.7813 0 4.5>, rotation=<rot 85 89.9987 -0.0015>}]
 dyn_R_sideCoat_2 [{shapeType=Capsule, length=11.5, radius=7, shapeName=dyn_L_sideCoat_1_capsule, center=<1.2595 0 7>, rotation=<rot 85 89.9987 -0.0015>}]
 dyn_R_sideCoat_3 [{shapeType=Capsule, length=18.875, radius=7, shapeName=dyn_L_sideCoat_1_capsule, center=<1.1433 0 5>, rotation=<rot 85 89.9987 -0.0015>}]
 dyn_R_sideCoat_base [{shapeType=Capsule, length=1, radius=1, shapeName=dyn_R_sideCoat_base_capsule, rotation=<rot 90 165.9638 -104.0362>}]
 foot_l [{
   shapeType=Capsule
   length=18.7278
   radius=9.6391
   shapeName=foot_l_capsule
   center=<-3.0517 7.328 0.0577>
   rotation=<rot 4.5442 -1.3816 -95.4352>
  }]
 foot_r [{
   shapeType=Capsule
   length=18.7278
   radius=9.6391
   shapeName=foot_r_capsule
   center=<3.0517 -7.328 -0.0577>
   rotation=<rot 4.5442 -1.3816 84.5648>
  }]
 pelvis [{shapeType=Capsule, length=5.1566, radius=8.375, shapeName=pelvis_capsule, center=<-1.646 -0.7954 -0.3869>, rotation=<rot 0 -0.0004 0>}]
 thigh_l []
 thigh_r []
}
constraints [
 {
  bone1Name=thigh_l
  bone2Name=pelvis
  swing1LimitDegrees=45
  swing1Motion=Free
  swing2LimitDegrees=45
  swing2Motion=Free
  twistLimitDegrees=45
  twistMotion=Free
 }
 {
  bone1Name=calf_l
  bone2Name=thigh_l
  swing1LimitDegrees=45
  swing1Motion=Free
  swing2LimitDegrees=45
  swing2Motion=Free
  twistLimitDegrees=45
  twistMotion=Free
 }
 {
  bone1Name=thigh_r
  bone2Name=pelvis
  swing1LimitDegrees=45
  swing1Motion=Free
  swing2LimitDegrees=45
  swing2Motion=Free
  twistLimitDegrees=45
  twistMotion=Free
 }
 {
  bone1Name=calf_r
  bone2Name=thigh_r
  swing1LimitDegrees=45
  swing1Motion=Free
  swing2LimitDegrees=45
  swing2Motion=Free
  twistLimitDegrees=45
  twistMotion=Free
 }
 {
  bone1Name=dyn_L_frontCoat_base
  bone2Name=pelvis
  swing1LimitDegrees=45
  swing1Motion=Free
  swing2LimitDegrees=45
  swing2Motion=Free
  twistLimitDegrees=45
  twistMotion=Free
 }
 {bone1Name=dyn_L_frontCoat_1, bone2Name=dyn_L_frontCoat_base, swing1LimitDegrees=0, swing2LimitDegrees=30, twistLimitDegrees=0}
 {bone1Name=dyn_L_frontCoat_2, bone2Name=dyn_L_frontCoat_1, swing1LimitDegrees=5, swing2LimitDegrees=30, twistLimitDegrees=0}
 {bone1Name=dyn_L_frontCoat_3, bone2Name=dyn_L_frontCoat_2, swing1LimitDegrees=0, swing2LimitDegrees=10, twistLimitDegrees=0}
 {
  bone1Name=dyn_L_sideCoat_base
  bone2Name=pelvis
  swing1LimitDegrees=45
  swing1Motion=Free
  swing2LimitDegrees=45
  swing2Motion=Free
  twistLimitDegrees=45
  twistMotion=Free
 }
 {bone1Name=dyn_L_sideCoat_1, bone2Name=dyn_L_sideCoat_base, swing1LimitDegrees=5, swing2LimitDegrees=30, twistLimitDegrees=0}
 {bone1Name=dyn_L_sideCoat_2, bone2Name=dyn_L_sideCoat_1, swing1LimitDegrees=5, swing2LimitDegrees=30, twistLimitDegrees=0}
 {bone1Name=dyn_L_sideCoat_3, bone2Name=dyn_L_sideCoat_2, swing1LimitDegrees=0, swing2LimitDegrees=30, twistLimitDegrees=0}
 {
  bone1Name=dyn_R_frontCoat_base
  bone2Name=pelvis
  swing1LimitDegrees=45
  swing1Motion=Free
  swing2LimitDegrees=45
  swing2Motion=Free
  twistLimitDegrees=45
  twistMotion=Free
 }
 {bone1Name=dyn_R_frontCoat_1, bone2Name=dyn_R_frontCoat_base, swing1LimitDegrees=0, swing2LimitDegrees=30, twistLimitDegrees=0}
 {bone1Name=dyn_R_frontCoat_2, bone2Name=dyn_R_frontCoat_1, swing1LimitDegrees=5, swing2LimitDegrees=30, twistLimitDegrees=0}
 {bone1Name=dyn_R_frontCoat_3, bone2Name=dyn_R_frontCoat_2, swing1LimitDegrees=0, swing2LimitDegrees=10, twistLimitDegrees=0}
 {
  bone1Name=dyn_R_sideCoat_base
  bone2Name=pelvis
  swing1LimitDegrees=45
  swing1Motion=Free
  swing2LimitDegrees=45
  swing2Motion=Free
  twistLimitDegrees=45
  twistMotion=Free
 }
 {bone1Name=dyn_R_sideCoat_1, bone2Name=dyn_R_sideCoat_base, swing1LimitDegrees=5, swing2LimitDegrees=30, twistLimitDegrees=0}
 {bone1Name=dyn_R_sideCoat_2, bone2Name=dyn_R_sideCoat_1, swing1LimitDegrees=5, swing2LimitDegrees=30, twistLimitDegrees=0}
 {bone1Name=dyn_R_sideCoat_3, bone2Name=dyn_R_sideCoat_2, swing1LimitDegrees=0, swing2LimitDegrees=30, twistLimitDegrees=0}
 {
  bone1Name=dyn_L_backCoat_base
  bone2Name=pelvis
  swing1LimitDegrees=45
  swing1Motion=Free
  swing2LimitDegrees=45
  swing2Motion=Free
  twistLimitDegrees=45
  twistMotion=Free
 }
 {bone1Name=dyn_L_backCoat_1, bone2Name=dyn_L_backCoat_base, swing1LimitDegrees=2, swing2LimitDegrees=20, twistLimitDegrees=0}
 {bone1Name=dyn_L_backCoat_2, bone2Name=dyn_L_backCoat_1, swing1LimitDegrees=5, swing2LimitDegrees=20, twistLimitDegrees=0}
 {bone1Name=dyn_L_backCoat_3, bone2Name=dyn_L_backCoat_2, swing1LimitDegrees=1, swing2LimitDegrees=20, twistLimitDegrees=0}
 {
  bone1Name=dyn_R_backCoat_base
  bone2Name=pelvis
  swing1LimitDegrees=45
  swing1Motion=Free
  swing2LimitDegrees=45
  swing2Motion=Free
  twistLimitDegrees=45
  twistMotion=Free
 }
 {bone1Name=dyn_R_backCoat_1, bone2Name=dyn_R_backCoat_base, swing1LimitDegrees=2, swing2LimitDegrees=20, twistLimitDegrees=0}
 {bone1Name=dyn_R_backCoat_2, bone2Name=dyn_R_backCoat_1, swing1LimitDegrees=5, swing2LimitDegrees=20, twistLimitDegrees=0}
 {bone1Name=dyn_R_backCoat_3, bone2Name=dyn_R_backCoat_2, swing1LimitDegrees=1, swing2LimitDegrees=20, twistLimitDegrees=0}
 {
  bone1Name=dyn_R_frontCoat_3
  bone2Name=dyn_R_sideCoat_3
  swing1LimitDegrees=45
  swing1Motion=Free
  swing2LimitDegrees=45
  swing2Motion=Free
  twistLimitDegrees=45
  twistMotion=Free
 }
 {
  bone1Name=foot_l
  bone2Name=calf_l
  swing1LimitDegrees=45
  swing1Motion=Free
  swing2LimitDegrees=45
  swing2Motion=Free
  twistLimitDegrees=45
  twistMotion=Free
 }
 {
  bone1Name=foot_r
  bone2Name=calf_r
  swing1LimitDegrees=45
  swing1Motion=Free
  swing2LimitDegrees=45
  swing2Motion=Free
  twistLimitDegrees=45
  twistMotion=Free
 }
 {
  bone1Name=dyn_L_sideCoat_3
  bone2Name=dyn_L_frontCoat_3
  swing1LimitDegrees=45
  swing1Motion=Free
  swing2LimitDegrees=45
  swing2Motion=Free
  twistLimitDegrees=45
  twistMotion=Free
 }
 {
  bone1Name=dyn_L_sideCoat_2
  bone2Name=dyn_L_frontCoat_2
  swing1LimitDegrees=45
  swing1Motion=Free
  swing2LimitDegrees=45
  swing2Motion=Free
  twistLimitDegrees=45
  twistMotion=Free
 }
 {
  bone1Name=dyn_R_frontCoat_3
  bone2Name=dyn_R_sideCoat_2
  swing1LimitDegrees=45
  swing1Motion=Free
  swing2LimitDegrees=45
  swing2Motion=Free
  twistLimitDegrees=45
  twistMotion=Free
 }
 {
  bone1Name=dyn_L_backCoat_3
  bone2Name=dyn_R_backCoat_3
  swing1LimitDegrees=45
  swing1Motion=Free
  swing2LimitDegrees=45
  swing2Motion=Free
  twistLimitDegrees=45
  twistMotion=Free
 }
 {
  bone1Name=dyn_R_backCoat_2
  bone2Name=dyn_R_sideCoat_2
  swing1LimitDegrees=45
  swing1Motion=Free
  swing2LimitDegrees=45
  swing2Motion=Free
  twistLimitDegrees=45
  twistMotion=Free
 }
 {
  bone1Name=dyn_R_sideCoat_3
  bone2Name=dyn_R_backCoat_3
  swing1LimitDegrees=45
  swing1Motion=Free
  swing2LimitDegrees=45
  swing2Motion=Free
  twistLimitDegrees=45
  twistMotion=Free
 }
]
