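# Control Rig for the M_MED_LabVapor face accessory mesh that offsets left and right dynamic hair bones based on head orientation.
# A head ModifyTransforms node with an identity transform runs first, followed by a spherical pose reader on the head bone with debug drawing enabled.
# The reader's output weights two ModifyTransforms nodes that apply preset rotations to the hair bones as a pose-driven corrective.
# The CrouchWalk and Crouching variables are declared but unused.
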ControlRigBlueprint /BRCosmetics/Characters/Player/Male/Medium/Bodies/M_MED_LabVapor/Meshes/Parts/M_MED_LabVapor_FaceAcc_CtrlRig
graphs [
 RigVMFunctionLibrary {links [], nodes []}
 RigVMModel {nodes [
   RigUnit_BeginExecution = &RigUnit_BeginExecution()
   RigVMFunction_Sequence = &RigVMFunction_Sequence(ExecuteContext=@RigUnit_BeginExecution.ExecutePin)
   &RigUnit_ModifyTransforms(
    ItemToModify=((Item=(Type=Bone,Name="head"),Transform=(Rotation=(X=0,Y=0,Z=0,W=1),Translation=(X=0,Y=0,Z=0),Scale3D=(X=1,Y=1,Z=1))))
    Weight=1
    WeightMinimum=0
    ExecutePin=@RigVMFunction_Sequence.A
   )
   > n1 = &RigUnit_SphericalPoseReader(
    ActiveRegionScaleFactors=(PositiveWidth=2,NegativeWidth=2,PositiveHeight=1.5,NegativeHeight=1.5)
    ActiveRegionSize=0.12
    Debug=(bDrawDebug=True,bDraw2D=False,bDrawLocalAxes=False,DebugScale=25,DebugSegments=20,DebugThickness=0.25)
    DriverAxis=(X=1,Y=0,Z=0)
    DriverItem=(Type=Bone,Name="head")
    FalloffRegionScaleFactors=(PositiveWidth=1.5,NegativeWidth=1,PositiveHeight=1.5,NegativeHeight=1.5)
    FalloffSize=0.3
    FlipHeightScaling=False
    FlipWidthScaling=False
    OptionalParentItem=(Type=Bone,Name="None")
    RotationOffset=(X=-90,Y=0,Z=40)
   )
   > &RigUnit_ModifyTransforms(
    ItemToModify="((Item=(Type=Bone,Name=\"dyn_Hair_R_C_Base\"),Transform=(Rotation=(X=0.0038,Y=-0.0871,Z=0.0435,W=0.9952),Translation=(X=0,Y=0,Z=0),Scale3D=(X=1,Y=1,Z=1))),(Item=(Type=Bone,Name=\"dyn_Hair_R_B_1\"),Transform=(Rotation=(X=-0.0038,Y=-0.0871,Z=-0.0435,W=0.9952),Tr ...(truncated)"
    WeightMinimum=0
    Weight=@n1.OutputParam
   )
   > &RigUnit_ModifyTransforms(
    ItemToModify="((Item=(Type=Bone,Name=\"dyn_Hair_L_A_1\"),Transform=(Rotation=(X=-0.0076,Y=-0.0868,Z=-0.0868,W=0.9924),Translation=(X=0,Y=0,Z=0),Scale3D=(X=1,Y=1,Z=1))),(Item=(Type=Bone,Name=\"dyn_Hair_L_B_1\"),Transform=(Rotation=(X=-0.0038,Y=-0.0871,Z=-0.0435,W=0.9952),Tra ...(truncated)"
    WeightMinimum=0
    Weight=@n1.OutputParam
   )
  ]}
]
hierarchy [
 root {loc=<0 0 0>, rot=<0 0 0 1>}
 pelvis {parent=root, loc=<0 -0.7071 93.1191>, rot=<0.0127 0.707 -0.0127 -0.707>}
 spine_01 {parent=pelvis, loc=<6.7494 -0.6388 0>, rot=<0 0 0.034 -0.9994>}
 spine_02 {parent=spine_01, loc=<7.2518 -0.035 0>, rot=<0 0 0.0238 -0.9997>}
 spine_03 {parent=spine_02, loc=<6.0104 0 0>, rot=<0 0 0.0149 -0.9999>}
 spine_04 {parent=spine_03, loc=<8.1024 0 0>, rot=<0 0 -0.0132 -0.9999>}
 spine_05 {parent=spine_04, loc=<10.0146 0 0>, rot=<0 0 -0.0015 -1>}
 neck_01 {parent=spine_05, loc=<11.3138 0.1529 0>, rot=<0 0 -0.2003 -0.9797>}
 neck_02 {parent=neck_01, loc=<4.5627 0 0>, rot=<0 0 0.0504 -0.9987>}
 head {parent=neck_02, loc=<5.3914 0 0>, rot=<0 0 0.115 -0.9934>}
 dyn_Hair_L_A_Base {parent=head, loc=<10.7722 5.9468 -6.5101>, rot=<-0.0029 0.7071 0.7071 -0.0029>}
 dyn_Hair_L_A_1 {parent=dyn_Hair_L_A_Base, loc=<0 0 0>, rot=<0 0 0 1>}
 dyn_Hair_L_A_2 {parent=dyn_Hair_L_A_1, loc=<3.6 0 0>, rot=<0 0 0 1>}
 dyn_Hair_L_A_3 {parent=dyn_Hair_L_A_2, loc=<3.6 0 0>, rot=<0 0 0 1>}
 dyn_Hair_L_A_4 {parent=dyn_Hair_L_A_3, loc=<3.6 0 0>, rot=<0 0 0 1>}
 dyn_Hair_L_B_Base {parent=head, loc=<10.8193 0.2486 -7.7979>, rot=<-0.0029 0.7071 0.7071 -0.0029>}
 dyn_Hair_L_B_1 {parent=dyn_Hair_L_B_Base, loc=<0 0 0>, rot=<0 0 0 1>}
 dyn_Hair_L_B_2 {parent=dyn_Hair_L_B_1, loc=<3.6 0 0>, rot=<0 0 0 1>}
 dyn_Hair_L_B_3 {parent=dyn_Hair_L_B_2, loc=<3.6 0 0>, rot=<0 0 0 1>}
 dyn_Hair_L_B_4 {parent=dyn_Hair_L_B_3, loc=<3.6 0 0>, rot=<0 0 0 1>}
 dyn_Hair_R_A_Base {parent=head, loc=<10.7726 5.9468 6.5101>, rot=<-0.0029 0.7071 0.7071 -0.0029>}
 dyn_Hair_R_A_1 {parent=dyn_Hair_R_A_Base, loc=<0 0 0>, rot=<0 0 0 1>}
 dyn_Hair_R_A_2 {parent=dyn_Hair_R_A_1, loc=<3.6 0 0>, rot=<0 0 0 1>}
 dyn_Hair_R_A_3 {parent=dyn_Hair_R_A_2, loc=<3.6 0 0>, rot=<0 0 0 1>}
 dyn_Hair_R_A_4 {parent=dyn_Hair_R_A_3, loc=<3.6 0 0>, rot=<0 0 0 1>}
 dyn_Hair_R_B_Base {parent=head, loc=<10.8197 0.2486 7.7979>, rot=<-0.0029 0.7071 0.7071 -0.0029>}
 dyn_Hair_R_B_1 {parent=dyn_Hair_R_B_Base, loc=<0 0 0>, rot=<0 0 0 1>}
 dyn_Hair_R_B_2 {parent=dyn_Hair_R_B_1, loc=<3.6 0 0>, rot=<0 0 0 1>}
 dyn_Hair_R_B_3 {parent=dyn_Hair_R_B_2, loc=<3.6 0 0>, rot=<0 0 0 1>}
 dyn_Hair_R_B_D {parent=dyn_Hair_R_B_3, loc=<3.6 0 0>, rot=<0 0 0 1>}
 dyn_Hair_R_C_Base {parent=head, loc=<10.8598 -4.6268 6.7289>, rot=<-0.0013 -0.7071 -0.7071 0.0072>}
 dyn_Hair_R_C_1 {parent=dyn_Hair_R_C_Base, loc=<0 0 0>, rot=<0 0 0 1>}
 dyn_Hair_R_C_2 {parent=dyn_Hair_R_C_1, loc=<3.5992 -0.0867 0>, rot=<0 0 0 1>}
 dyn_Hair_R_C_3 {parent=dyn_Hair_R_C_2, loc=<3.5992 -0.0867 0>, rot=<0 0 0 1>}
 dyn_Hair_R_C_4 {parent=dyn_Hair_R_C_3, loc=<3.5992 -0.0867 0>, rot=<0 0 0 1>}
 dyn_Hair_L_C_Base {parent=head, loc=<10.8594 -4.6268 -6.8011>, rot=<-0.0013 -0.7071 -0.7071 0.0072>}
 dyn_Hair_L_C_1 {parent=dyn_Hair_L_C_Base, loc=<0 0 0>, rot=<0 0 0 1>}
 dyn_Hair_L_C_2 {parent=dyn_Hair_L_C_1, loc=<3.6 0 0>, rot=<0 0 0 1>}
 dyn_Hair_L_C_3 {parent=dyn_Hair_L_C_2, loc=<3.6 0 0>, rot=<0 0 0 1>}
 dyn_Hair_L_C_4 {parent=dyn_Hair_L_C_3, loc=<3.6 0 0>, rot=<0 0 0 1>}
]
variables [
 CrouchWalk {type=double, default=0}
 Crouching {type=bool, default=False}
]
