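# Physics asset for the M_MED_LabVapor player face accessory, covering hair strand physics with head, neck and outer collar collision shapes.
# Six capsule chains (A, B, C per side) hang from Base bodies locked to head, itself locked to neck_02.
# Links along each chain carry swing angle settings, tighter toward B and C tips, but no locked axes.
# Most first-link bodies have no shapes, and the three collar bodies have no constraints.
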PhysicsAsset /BRCosmetics/Characters/Player/Male/Medium/Bodies/M_MED_LabVapor/Meshes/Parts/M_MED_LabVapor_FaceAcc_Physics
bodies {
 dyn_Hair_L_A_1 [{shapeType=Capsule, length=2.8955, radius=1.142, shapeName=dyn_Hair_L_A_1_capsule, center=<1 0 0.3864>, rotation=<rot 0 90 -90>}]
 dyn_Hair_L_A_2 [{shapeType=Capsule, length=3.2175, radius=0.7825, shapeName=dyn_Hair_L_A_1_capsule, center=<1.9968 0 0>, rotation=<rot 0 89.6511 -89.9644>}]
 dyn_Hair_L_A_3 [{shapeType=Capsule, length=4.02, radius=0.78, shapeName=dyn_Hair_L_A_1_capsule, center=<2.9975 0 -0.0025>, rotation=<rot 0 89.6511 -89.9644>}]
 dyn_Hair_L_A_4 [{shapeType=Capsule, length=4.3, radius=0.8, shapeName=dyn_Hair_L_A_1_capsule, center=<4 0 -0.002>, rotation=<rot 0 89.6511 -89.9644>}]
 dyn_Hair_L_A_Base [{shapeType=Capsule, length=0.505, radius=0.505, shapeName=dyn_Hair_L_A_Base_capsule, rotation=<rot 0 -90 0>}]
 dyn_Hair_L_B_1 []
 dyn_Hair_L_B_2 [{shapeType=Capsule, length=2.4, radius=1.3, shapeName=dyn_Hair_L_A_1_capsule, center=<1.2423 0 -0.0025>, rotation=<rot 0 89.6511 -89.9644>}]
 dyn_Hair_L_B_3 [{shapeType=Capsule, length=2.4, radius=1, shapeName=dyn_Hair_L_A_1_capsule, center=<2 0 -0.0025>, rotation=<rot 0 89.6511 -89.9644>}]
 dyn_Hair_L_B_4 [{
   shapeType=Capsule
   length=5
   radius=0.75
   shapeName=dyn_Hair_L_A_1_capsule
   center=<3.3502 1.1545 0.6079>
   rotation=<rot -79.3409 15.0916 0.0001>
  }]
 dyn_Hair_L_B_Base [{shapeType=Capsule, length=0.505, radius=0.505, shapeName=dyn_Hair_L_B_Base_capsule, rotation=<rot 0 -90 0>}]
 dyn_Hair_L_C_1 []
 dyn_Hair_L_C_2 [{shapeType=Capsule, length=2.4, radius=1.3, shapeName=dyn_Hair_L_A_1_capsule, center=<1.2423 0 -0.0025>, rotation=<rot 0 89.6511 -89.9644>}]
 dyn_Hair_L_C_3 [{shapeType=Capsule, length=2.4, radius=1, shapeName=dyn_Hair_L_A_1_capsule, center=<1.2504 0.667 -0.0025>, rotation=<rot 0 89.6511 -89.9644>}]
 dyn_Hair_L_C_4 [{
   shapeType=Capsule
   length=6.6767
   radius=0.6
   shapeName=dyn_Hair_L_A_1_capsule
   center=<3.1595 0.9506 0.7067>
   rotation=<rot -0.0118 108.903 -79.3266>
  }]
 dyn_Hair_L_C_Base [{shapeType=Capsule, length=0.505, radius=0.505, shapeName=dyn_Hair_L_C_Base_capsule, rotation=<rot 0 -90 0>}]
 dyn_Hair_R_A_1 []
 dyn_Hair_R_A_2 [{shapeType=Capsule, length=3, radius=1, shapeName=dyn_Hair_R_A_1_capsule, center=<2 0 0.0025>, rotation=<rot 0 89.6511 90.0356>}]
 dyn_Hair_R_A_3 [{shapeType=Capsule, length=4, radius=0.8, shapeName=dyn_Hair_R_A_1_capsule, center=<3 0 0.0025>, rotation=<rot 0 89.6511 90.0356>}]
 dyn_Hair_R_A_4 [{shapeType=Capsule, length=4.3, radius=0.8, shapeName=dyn_Hair_R_A_1_capsule, center=<4 0 0.002>, rotation=<rot 0 89.6511 90.0356>}]
 dyn_Hair_R_A_Base [{shapeType=Capsule, length=0.505, radius=0.505, shapeName=dyn_Hair_R_A_Base_capsule, rotation=<rot 0 -90 0>}]
 dyn_Hair_R_B_1 []
 dyn_Hair_R_B_2 [{shapeType=Capsule, length=2.65, radius=1.05, shapeName=dyn_Hair_R_A_1_capsule, center=<1.6635 0 0.0043>, rotation=<rot 0 89.6511 90.0356>}]
 dyn_Hair_R_B_3 [{shapeType=Capsule, length=2.4, radius=1, shapeName=dyn_Hair_L_A_1_capsule, center=<1.2423 0 -0.0025>, rotation=<rot 0 89.6511 -89.9644>}]
 dyn_Hair_R_B_Base [{shapeType=Capsule, length=0.505, radius=0.505, shapeName=dyn_Hair_R_B_Base_capsule, rotation=<rot 0 -90 0>}]
 dyn_Hair_R_B_D [{
   shapeType=Capsule
   length=5
   radius=0.8
   shapeName=dyn_Hair_R_B_D_capsule
   center=<3.2341 -0.9023 0.5013>
   rotation=<rot -2.8126 -101.3275 -104.1243>
  }]
 dyn_Hair_R_C_1 []
 dyn_Hair_R_C_2 [{
   shapeType=Capsule
   length=2.4
   radius=1.3
   shapeName=dyn_Hair_R_A_1_capsule
   center=<0.9138 -0.0131 0.0038>
   rotation=<rot 0 89.6511 90.0356>
  }]
 dyn_Hair_R_C_3 [{shapeType=Capsule, length=2.4, radius=1, shapeName=dyn_Hair_R_A_1_capsule, center=<0.8685 -0.0129 0.0038>, rotation=<rot 0 89.6511 90.0356>}]
 dyn_Hair_R_C_4 [{
   shapeType=Capsule
   length=6
   radius=0.6
   shapeName=dyn_Hair_R_A_1_capsule
   center=<2.8489 -0.4018 0.2487>
   rotation=<rot 5.5422 77.8538 99.136>
  }]
 dyn_Hair_R_C_Base [{shapeType=Capsule, length=0.505, radius=0.505, shapeName=dyn_Hair_R_C_Base_capsule, rotation=<rot 0 -90 0>}]
 dyn_outerCollar_01_c [
  {
   shapeType=Capsule
   length=14.9127
   radius=6
   shapeName=dyn_outerCollar_01_c_capsule
   center=<1.3838 5.7007 -5.1822>
   rotation=<rot 0 14.1579 -90>
  }
  {
   shapeType=Capsule
   length=18.1469
   radius=6
   shapeName=dyn_outerCollar_01_c_capsule
   center=<1.5322 -4.1701 -5.6848>
   rotation=<rot 58.3825 0 104.2198>
  }
 ]
 dyn_outerCollar_01_l [{
   shapeType=Capsule
   length=9.0642
   radius=6.9358
   shapeName=dyn_outerCollar_01_l_capsule
   center=<5.5183 2.4644 0.008>
   rotation=<rot 9.2616 15.6172 57.659>
  }]
 dyn_outerCollar_01_r [
  {
   shapeType=Capsule
   length=5.9321
   radius=6.9167
   shapeName=dyn_outerCollar_01_r_capsule
   center=<8.1054 0.0713 -5.7694>
   rotation=<rot -36.118 -16.1278 132.687>
  }
  {
   shapeType=Capsule
   length=9.2921
   radius=6.1122
   shapeName=dyn_outerCollar_01_r_capsule
   center=<4.3899 -5.6188 4.42>
   rotation=<rot -11.3203 -37.2206 145.3919>
  }
 ]
 head [
  {
   shapeType=Capsule
   length=9.6594
   radius=6.4479
   shapeName=head_capsule
   center=<3.3274 1.1486 0.5002>
   rotation=<rot -0.01 -91.7455 -89.9917>
  }
  {
   shapeType=Capsule
   length=10.4738
   radius=5.6262
   shapeName=head_capsule
   center=<3.2146 2.6307 -0.7443>
   rotation=<rot -2.8913 -89.8674 -89.8382>
  }
  {
   shapeType=Capsule
   length=14.3501
   radius=6.0828
   shapeName=head_capsule
   center=<-0.0296 0.2155 -0.8209>
   rotation=<rot -87.9371 179.9999 -180>
  }
 ]
 neck_02 [{shapeType=Capsule, length=6.4701, radius=6.2799, shapeName=neck_02_capsule, center=<0.4304 2.1331 0>, rotation=<rot 0 -105.0005 90>}]
}
constraints [
 {
  bone1Name=dyn_Hair_L_A_Base
  bone2Name=head
  swing1LimitDegrees=45
  swing1Motion=Locked
  swing2LimitDegrees=45
  swing2Motion=Locked
  twistLimitDegrees=45
  twistMotion=Locked
 }
 {bone1Name=dyn_Hair_L_A_1, bone2Name=dyn_Hair_L_A_Base, swing1LimitDegrees=45, swing2LimitDegrees=45, twistLimitDegrees=0}
 {bone1Name=dyn_Hair_L_A_2, bone2Name=dyn_Hair_L_A_1, swing1LimitDegrees=45, swing2LimitDegrees=45, twistLimitDegrees=0}
 {bone1Name=dyn_Hair_L_A_3, bone2Name=dyn_Hair_L_A_2, swing1LimitDegrees=45, swing2LimitDegrees=45, twistLimitDegrees=0}
 {bone1Name=dyn_Hair_L_A_4, bone2Name=dyn_Hair_L_A_3, swing1LimitDegrees=45, swing2LimitDegrees=45, twistLimitDegrees=0}
 {
  bone1Name=dyn_Hair_L_B_Base
  bone2Name=head
  swing1LimitDegrees=45
  swing1Motion=Locked
  swing2LimitDegrees=45
  swing2Motion=Locked
  twistLimitDegrees=45
  twistMotion=Locked
 }
 {bone1Name=dyn_Hair_L_B_1, bone2Name=dyn_Hair_L_B_Base, swing1LimitDegrees=45, swing2LimitDegrees=45, twistLimitDegrees=0}
 {bone1Name=dyn_Hair_L_B_2, bone2Name=dyn_Hair_L_B_1, swing1LimitDegrees=45, swing2LimitDegrees=45, twistLimitDegrees=0}
 {bone1Name=dyn_Hair_L_B_3, bone2Name=dyn_Hair_L_B_2, swing1LimitDegrees=10, swing2LimitDegrees=10, twistLimitDegrees=0}
 {bone1Name=dyn_Hair_L_B_4, bone2Name=dyn_Hair_L_B_3, swing1LimitDegrees=10, swing2LimitDegrees=10, twistLimitDegrees=0}
 {
  bone1Name=dyn_Hair_L_C_Base
  bone2Name=head
  swing1LimitDegrees=45
  swing1Motion=Locked
  swing2LimitDegrees=45
  swing2Motion=Locked
  twistLimitDegrees=45
  twistMotion=Locked
 }
 {bone1Name=dyn_Hair_L_C_1, bone2Name=dyn_Hair_L_C_Base, swing1LimitDegrees=30, swing2LimitDegrees=30, twistLimitDegrees=0}
 {bone1Name=dyn_Hair_L_C_2, bone2Name=dyn_Hair_L_C_1, swing1LimitDegrees=30, swing2LimitDegrees=30, twistLimitDegrees=0}
 {bone1Name=dyn_Hair_L_C_3, bone2Name=dyn_Hair_L_C_2, swing1LimitDegrees=7, swing2LimitDegrees=7, twistLimitDegrees=0}
 {bone1Name=dyn_Hair_L_C_4, bone2Name=dyn_Hair_L_C_3, swing1LimitDegrees=7, swing2LimitDegrees=7, twistLimitDegrees=0}
 {
  bone1Name=dyn_Hair_R_A_Base
  bone2Name=head
  swing1LimitDegrees=45
  swing1Motion=Locked
  swing2LimitDegrees=45
  swing2Motion=Locked
  twistLimitDegrees=45
  twistMotion=Locked
 }
 {bone1Name=dyn_Hair_R_A_1, bone2Name=dyn_Hair_R_A_Base, swing1LimitDegrees=45, swing2LimitDegrees=45, twistLimitDegrees=0}
 {bone1Name=dyn_Hair_R_A_2, bone2Name=dyn_Hair_R_A_1, swing1LimitDegrees=45, swing2LimitDegrees=45, twistLimitDegrees=0}
 {bone1Name=dyn_Hair_R_A_3, bone2Name=dyn_Hair_R_A_2, swing1LimitDegrees=45, swing2LimitDegrees=45, twistLimitDegrees=0}
 {bone1Name=dyn_Hair_R_A_4, bone2Name=dyn_Hair_R_A_3, swing1LimitDegrees=45, swing2LimitDegrees=45, twistLimitDegrees=0}
 {
  bone1Name=dyn_Hair_R_B_Base
  bone2Name=head
  swing1LimitDegrees=45
  swing1Motion=Locked
  swing2LimitDegrees=45
  swing2Motion=Locked
  twistLimitDegrees=45
  twistMotion=Locked
 }
 {bone1Name=dyn_Hair_R_B_1, bone2Name=dyn_Hair_R_B_Base, swing1LimitDegrees=45, swing2LimitDegrees=45, twistLimitDegrees=0}
 {bone1Name=dyn_Hair_R_B_2, bone2Name=dyn_Hair_R_B_1, swing1LimitDegrees=45, swing2LimitDegrees=45, twistLimitDegrees=0}
 {bone1Name=dyn_Hair_R_B_3, bone2Name=dyn_Hair_R_B_2, swing1LimitDegrees=10, swing2LimitDegrees=10, twistLimitDegrees=0}
 {bone1Name=dyn_Hair_R_B_D, bone2Name=dyn_Hair_R_B_3, swing1LimitDegrees=10, swing2LimitDegrees=10, twistLimitDegrees=0}
 {
  bone1Name=dyn_Hair_R_C_Base
  bone2Name=head
  swing1LimitDegrees=45
  swing1Motion=Locked
  swing2LimitDegrees=45
  swing2Motion=Locked
  twistLimitDegrees=45
  twistMotion=Locked
 }
 {bone1Name=dyn_Hair_R_C_1, bone2Name=dyn_Hair_R_C_Base, swing1LimitDegrees=45, swing2LimitDegrees=45, twistLimitDegrees=0}
 {bone1Name=dyn_Hair_R_C_2, bone2Name=dyn_Hair_R_C_1, swing1LimitDegrees=45, swing2LimitDegrees=45, twistLimitDegrees=0}
 {bone1Name=dyn_Hair_R_C_3, bone2Name=dyn_Hair_R_C_2, swing1LimitDegrees=10, swing2LimitDegrees=10, twistLimitDegrees=0}
 {bone1Name=dyn_Hair_R_C_4, bone2Name=dyn_Hair_R_C_3, swing1LimitDegrees=10, swing2LimitDegrees=10, twistLimitDegrees=0}
 {
  bone1Name=head
  bone2Name=neck_02
  swing1LimitDegrees=45
  swing1Motion=Locked
  swing2LimitDegrees=45
  swing2Motion=Locked
  twistLimitDegrees=45
  twistMotion=Locked
 }
]
